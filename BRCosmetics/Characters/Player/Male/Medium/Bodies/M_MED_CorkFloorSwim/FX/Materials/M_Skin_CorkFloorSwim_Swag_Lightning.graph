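# Unlit translucent Niagara particle material drawing animated lightning bolts for the CorkFloorSwim Swag skin effect.
# A randomly chosen bolt column pans on choppy, per-particle time with noise UV distortion, then is masked and shaped into core and glow opacity with depth fade.
# World position offset thickens spline meshes toward the camera using dynamic-parameter width, with an Android scale fix.
# Particle color drives emissive and alpha; a previous-frame switch adjusts the offset by pan motion.
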Material /BRCosmetics/Characters/Player/Male/Medium/Bodies/M_MED_CorkFloorSwim/FX/Materials/M_Skin_CorkFloorSwim_Swag_Lightning
graph {
 properties {
  bAutomaticallySetUsageInEditor=false
  bUsedWithNiagaraMeshParticles=true
  bUsedWithNiagaraSprites=true
  blendMode=BLEND_Translucent
  opacityMaskClipValue=0.1
  pixelDepthOffsetMode=PDOM_Legacy
  shadingModel=MSM_Unlit
  twoSided=true
  usedShadingModels=Unlit
 }
 nodes [
  n1 = ParticleColor()
  MaterialFunctionCall(Emissive=@n1.RGB) {materialFunction=&MF_FX_EmissiveAdjustments}
  n2 = ScalarParameter() {defaultValue=3, parameterName=Depth Fade Distance}
  n3 = Constant() {r=1}
  n4 = NamedRerouteUsage()
  n5 = ScalarParameter() {defaultValue=2, parameterName=Lightning Glow Exponent}
  n6 = OneMinus(None=Multiply(A=Power(Base=@n4, Exp=@n5), B=ScalarParameter() {defaultValue=0.1, parameterName=Lightning Glow Intensity}))
  n7 = Saturate(None=ScalarParameter() {defaultValue=0.2, parameterName=Lightning Core Size})
  n8 = Subtract(A=OneMinus(None=NamedRerouteUsage()), B=@n7)
  n9 = OneMinus(None=@n7)
  n10 = Saturate(None=ScalarParameter() {defaultValue=0.8, parameterName=Lightning Core Softness})
  n11 = Saturate(None=Divide(A=Saturate(None=@n8), B=Multiply(A=@n9, B=@n10)))
  n12 = ScalarParameter() {defaultValue=5, parameterName=Lightning Core Exponent}
  n13 = OneMinus(None=Power(Base=OneMinus(None=@n11), Exp=@n12))
  n14 = Multiply(A=QualitySwitch(Default=DepthFade(FadeDistance=@n2), Low=@n3, Medium=@n3), B=OneMinus(None=Multiply(A=@n6, B=@n13)))
  n15 = NamedRerouteUsage()
  Multiply(A=@n1.A, B=NamedRerouteDeclaration(None=Multiply(A=@n14, B=@n15)) {name=Opacity, nodeColor=<rgba 1 0 0.9375 1>})
  n16 = TextureCoordinate()
  n17 = ComponentMask(None=@n16) {g=true}
  n18 = DynamicParameter() {defaultValue=<rgba 10 0 0 0>, paramNames [Lightning Width, Lightning Tiling, Lightning Pan, ""]}
  n19 = Saturate(None=Divide(A=Subtract(A=@n17, B=@n18.Lightning Pan), B=ScalarParameter() {defaultValue=0.333, parameterName=Opacity - Fade Size Tip}))
  n20 = Saturate(None=Divide(A=OneMinus(None=@n17), B=ScalarParameter() {defaultValue=0.2, parameterName=Opacity - Fade Size Base}))
  Multiply(A=@n19, B=@n20)
  n21 = TextureCoordinate()
  Multiply(A=OneMinus(None=ComponentMask(None=@n21) {g=true}), B=@n12)
  n22 = NamedRerouteUsage()
  n23 = NamedRerouteUsage()
  n24 = VectorParameter() {parameterName=UV Noise Scale and Seped, defaultValue=<rgba 0.4 0.4 0.4 1.3>}
  n25 = Panner(
   Coordinate=Multiply(A=TextureCoordinate(), B=ComponentMask(None=@n24.RGB) {g=true, r=true})
   Time=NamedRerouteUsage()
   Speed=AppendVector(A=@n24.B, B=@n24.B)
  ) {bFractionalPart=true, speedY=1}
  n26 = Multiply(
   A=ConstantBiasScale(None=TextureSample(UVs=@n25) {samplerType=SAMPLERTYPE_LinearColor, texture=&HighContrastNoise}.R) {bias=-0.5, scale=2}
   B=NamedRerouteUsage()
  )
  n27 = ScalarParameter() {defaultValue=0.03, parameterName=UV Distortion Intensity}
  n28 = Multiply(A=ScalarParameter() {defaultValue=1.3, parameterName=Lightning UV Width Scale}, B=Multiply(A=NamedRerouteUsage(), B=NamedRerouteUsage()))
  n29 = MaterialFunctionCall(Float2=MaterialFunctionCall(UVs=TextureCoordinate(), Texture Scale=AppendVector(A=@n28, B=@n18.Lightning Tiling)) {materialFunction=&ScaleUVsByCenter}.UVs) {materialFunction=&BreakOutFloat2Components}
  n30 = Saturate(None=@n29.R)
  n31 = Constant()
  n32 = AppendVector(A=@n31, B=Divide(A=ScalarParameter() {defaultValue=2, parameterName=Lightning Texture Pan Speed}, B=@n18.Lightning Tiling))
  n33 = Panner(
   Coordinate=MaterialFunctionCall(
    "Animation  Phase (0-1)"=ParticleRandom()
    Number of Rows=Constant() {r=1}
    Number of Columns=Constant() {r=3}
    UVs=MaterialFunctionCall(X=@n30, Y=@n29.G) {materialFunction=&MakeFloat2}.Result
   ) {materialFunction=&FlipBook}.UVs
   Time=NamedRerouteUsage()
   Speed=@n32
  ) {bFractionalPart=true, speedY=1.1}
  n34 = TextureSample(UVs=Add(A=Multiply(A=@n26, B=@n27), B=@n33)) {
   samplerType=SAMPLERTYPE_LinearGrayscale
   texture=/BRCosmetics/Athena/Effects/Particles/Skydiving/NS_Contrail_GreatPool/Meshes/T_FX_GreatPool_Bolts
  }
  NamedRerouteDeclaration(None=Multiply(A=Multiply(A=@n22, B=@n23), B=@n34.R) {constA=1}) {name=LightningSDF, nodeColor=<rgba 0.8047 1 0 1>}
  NamedRerouteDeclaration(None=@n19) {name=Length Mask, nodeColor=<rgba 1 0 0.8437 1>}
  n35 = ComponentMask(None=@n16) {r=true}
  n36 = OneMinus(None=Abs(None=ConstantBiasScale(None=@n35) {bias=-0.5, scale=2}))
  n37 = ScalarParameter() {defaultValue=0.133, parameterName=Opacity - Fade Size Width}
  NamedRerouteDeclaration(None=Saturate(None=Divide(A=@n36, B=@n37))) {name=Width Mask, nodeColor=<rgba 0 0.2969 1 1>}
  n38 = WorldPosition()
  n39 = NamedRerouteUsage()
  n40 = Normalize(VectorInput=Subtract(A=CameraPositionWS(), B=Add(A=@n38.XYZ, B=@n39)))
  n41 = ScalarParameter() {defaultValue=3, parameterName=Camera Offset}
  NamedRerouteDeclaration(None=Add(A=Multiply(A=@n40, B=@n41), B=NamedRerouteUsage())) {name=WPO, nodeColor=<rgba 1 0 0.5391 1>}
  n42 = AppendVector(A=@n32, B=Constant())
  n43 = DeltaTime()
  NamedRerouteDeclaration(None=Transform(None=Multiply(A=@n42, B=@n43))) {name=Panner Vector, nodeColor=<rgba 1 0 0.7969 1>}
  n44 = ScalarParameter() {defaultValue=19, parameterName=UV Panner Flicker Frequency}
  n45 = ParticleRandom()
  n46 = Add(A=Time(), B=Frac(None=Multiply(A=@n45) {constB=3.618})) {constB=0}
  n47 = Ceil(None=Multiply(A=@n46, B=@n44) {constB=4})
  n48 = ScalarParameter() {defaultValue=2.62, parameterName=Time Choppiness Frequency}
  n49 = Multiply(A=@n46, B=@n48)
  n50 = ComponentMask(None=VectorNoise(World Position=@n49)) {r=true}
  n51 = Add(A=Multiply(A=Subtract(A=@n50) {constB=0.5}, B=ScalarParameter() {defaultValue=0.9, parameterName=Time Choppiness Randomness Range})) {constB=0.5}
  n52 = Floor(None=@n49)
  n53 = Divide(A=Subtract(A=Subtract(A=@n49, B=@n51), B=@n52), B=OneMinus(None=@n51))
  n54 = Divide(A=Max(A=Add(A=@n53, B=@n52), B=@n52), B=@n48)
  n55 = Divide(A=Subtract(A=@n46, B=@n54), B=@n51)
  NamedRerouteDeclaration(None=LinearInterpolate(A=Divide(A=@n47, B=@n44), B=@n54, Alpha=Saturate(None=@n55)) {desc=Blending between quantized time and choppy time}) {name=Choppy Time, nodeColor=<rgba 0 1 0.9297 1>}
  NamedRerouteDeclaration(None=@n20) {name=Base Mask, nodeColor=<rgba 1 0.9375 0 1>}
  n56 = Normalize(VectorInput=Subtract(A=CameraPositionWS(), B=WorldPosition().XYZ))
  n57 = Transform(None=Constant3Vector() {constant=<rgba 0 1 0 0>})
  n58 = DynamicParameter() {
   desc="This only exists to fix the splinethicken meshes on android,\r\napparently the tangents do not properly respond to negative scale there"
   parameterIndex=1
   defaultValue=<rgba 1 1 1 0>
   paramNames [Scale X, Scale Y, Scale Z, ""]
  }
  n59 = ShadingPathSwitch(Default=@n57, Deferred=@n57, Mobile=Multiply(A=@n57, B=@n58.RGB)) {desc=Migrated from FeatureLevelSwitch}
  n60 = CrossProduct(A=@n56, B=@n59)
  n61 = TextureCoordinate()
  n62 = Multiply(
   A=Subtract(A=ComponentMask(None=@n61) {r=true}) {constB=0.5}
   B=DynamicParameter() {defaultValue=<rgba 10 0 0 0>, paramNames [Lightning Width, Lightning Tiling, Lightning Pan, ""]}.Lightning Width
  )
  n63 = Abs(None=DotProduct(A=@n56, B=@n59))
  n64 = NamedRerouteDeclaration(None=SquareRoot(None=OneMinus(None=@n63))) {name=Fresnel Fade, nodeColor=<rgba 0 0.9297 1 1>}
  NamedRerouteDeclaration(None=Multiply(A=Normalize(VectorInput=@n60), B=Multiply(A=@n62, B=@n64))) {name=SplineThicken WPO, nodeColor=<rgba 0.2891 0 1 1>}
  n65 = NamedRerouteUsage()
  PreviousFrameSwitch(Current Frame=@n65, Previous Frame=Add(A=@n65, B=NamedRerouteUsage()))
  Subtract(A=@n33, B=AppendVector(A=@n31, B=Divide(A=@n18.Lightning Pan, B=@n18.Lightning Pan))) {desc=Feeling cute might delete later}
 ]
}
