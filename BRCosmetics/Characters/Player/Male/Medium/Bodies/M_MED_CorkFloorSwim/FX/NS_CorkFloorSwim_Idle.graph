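# Idle Niagara effect for the CorkFloorSwim player body cosmetic that keeps fire burning at multiple body sockets.
# FireMain mesh flames, FireParticle flipbook sprites, curl-noise Embers and flickering Glow sprites with lights use partition-indexed arrays for per-socket placement and tuning.
# A system timeline pulses emission, size or brightness, with time-of-day brightness scaling on most layers.
# User booleans selectively hide partitions for ADS, hidden hand or head, and back or eye FX.
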NiagaraSystem /BRCosmetics/Characters/Player/Male/Medium/Bodies/M_MED_CorkFloorSwim/FX/NS_CorkFloorSwim_Idle
dependencies {
 usedDataInterfaces [
  &NiagaraDataInterfaceCurve
  &NiagaraDataInterfaceArrayFloat3
  &NiagaraDataInterfaceArrayFloat
  &NiagaraDataInterfaceColorCurve
  &NiagaraDataInterfaceVectorCurve
  &NiagaraDataInterfacePlatformSet
  &NiagaraDataInterfaceArrayInt32
  &NiagaraDataInterfaceSkeletalMesh
  &NiagaraDataInterfaceSocketReader
 ]
 usedDynamicInputs [
  &Multiply_Vector2D_ByFloat
  &Lerp_Float
  &FNTimeOfDayScalar
 ]
 usedModules [
  &self:SystemSpawnScript.SystemScriptSource.SystemScriptGraph.NiagaraNodeAssignment_0.SetVariables_#1
  &self:SystemSpawnScript.SystemScriptSource.SystemScriptGraph.NiagaraNodeAssignment_5.SetVariables_#2
  &self:SystemSpawnScript.SystemScriptSource.SystemScriptGraph.NiagaraNodeAssignment_1.SetVariables_#3
  &SystemState
  &self:SystemSpawnScript.SystemScriptSource.SystemScriptGraph.NiagaraNodeAssignment_2.SetVariables_#4
  &self:SystemSpawnScript.SystemScriptSource.SystemScriptGraph.NiagaraNodeAssignment_3.SetVariables_#5
  &Timeline
  &self:SystemSpawnScript.SystemScriptSource.SystemScriptGraph.NiagaraNodeAssignment_4.SetVariables_#6
 ]
}
emitters {
 Embers {
  emitterUpdateScript [
   &EmitterState(
    Enable Distance Culling=true
    Inactive Response=::NewEnumerator1
    Life Cycle Mode=::NewEnumerator1
    Loop Duration=1
    MaxDistance=2500
    Module.bApplySpawnCountScaleByDistance=true
    Module.bUseMaxDistance=true
    Module.bUseMinDistance=false
    Scalability Mode=::NewEnumerator1
    Scale Spawn Count=true
    Spawn Count Scale By Distance=DataInterface() {propertyValues {Curve {keys [<at 0 1>,<at 0.5 1>,<at 1 0>]}}}
    UseLoopDelay=false
   )
   &SpawnRate(SpawnRate=55, Use Spawn Probability=false)
   &SpawnRate(
    SpawnRate=&FloatFromCurve(FloatCurve=DataInterface() {propertyValues {Curve {keys [<at 0 1 RCIM_Cubic>,<at 1 0>]}}}, CurveIndex=$System.Timeline.Time, Scale Curve=55)
    Use Spawn Probability=false
   ) {module=SpawnRate001}
  ]
  particleSpawnScript [
   &PartitionParticles(Number Of Partitions=12)
   &InitializeParticle(
    Color=&RandomRangeLinearColor(Minimum=<rgba 14 11.3989 0.0479 1>, Maximum=<rgba 14 3.886 0 1>)
    Lifetime Max=0.7
    Lifetime Min=0.4
    Lifetime Mode=::NewEnumerator1
    Sprite Size Max=<1.7 3>
    Sprite Size Min=<0.8 2>
    Sprite Size Mode=::NewEnumerator2
    UsePositionOffset=true
    Write Material Random=true
   )
   &SkeletalMeshLocation(
    Bone / Socket Index=$Particles.PartitionParticles.Partition
    Bone Sampling Mode=::NewEnumerator4
    MaskSpawnGroup=false
    Sampled Position Offset=&Add_Vector(
     A=&RandomRangeVector(Minimum=<-6 -6 -6>, Maximum=<6 6 6>)
     B=&SelectVectorFromArray(
      Direct Array Index=$Particles.PartitionParticles.Partition
      Vector Selection Array=DataInterface() {propertyValues {FloatData [
         <4 12 9>
         <4 12 -9>
         <-6 10 2>
         <-6 11 -2>
         <7.5 0 -1>
         <-7.5 0 1>
         <5.5 2 1.5>
         <5.5 2 -1.5>
         <-22 0 0>
         <4 8.6 2>
         <4.3 8.6 -2>
         <6.3432 16.9792 0>
        ]}}
     )
    )
    Skeletal Mesh=$System.Skeletal Mesh
   ) {enabled=false, module=SkeletalMeshLocation001}
   &SocketLocation(
    SamplePosOffset=true
    Sampled Position Offset=&Add_Vector(
     A=&RandomRangeVector(Minimum=<-6 -6 -6>, Maximum=<6 6 6>)
     B=&SelectVectorFromArray(
      Direct Array Index=$Particles.PartitionParticles.Partition
      Vector Selection Array=DataInterface() {propertyValues {FloatData [
         <4 12 9>
         <4 12 -9>
         <-6 10 2>
         <-6 11 -2>
         <7.5 0 -1>
         <-7.5 0 1>
         <5.5 2 1.5>
         <5.5 2 -1.5>
         <-22 0 0>
         <4 8.6 2>
         <4.3 8.6 -2>
         <6.3432 16.9792 0>
        ]}}
     )
    )
    Socket Index=$Particles.PartitionParticles.Partition
    Socket Reader=$System.Socket Reader
    Socket Sampling Mode=::NewEnumerator3
    VelocitySpeedLimit=false
   )
   &AddVelocity(Velocity=&RandomRangeVector(Minimum=<0 0 5>, Maximum=<0 0 25>))
  ]
  particleUpdateScript [
   &KillParticles(Kill Particles=&Boolean_AndOperation(
     Bool1=$User.bIsHandHidden
     Bool2=&Boolean_AndOperation(
      Bool1=&SetBoolByIntComparison(A=$Particles.PartitionParticles.Partition, B=6, Comparison Type=::NewEnumerator2)
      Bool2=&SetBoolByIntComparison(A=$Particles.PartitionParticles.Partition, B=8, Comparison Type=::NewEnumerator7)
     )
    ))
   &KillParticles(Kill Particles=&Boolean_AndOperation(
     Bool1=$User.bIsHandHidden
     Bool2=&Boolean_AndOperation(
      Bool1=&SetBoolByIntComparison(A=$Particles.PartitionParticles.Partition, B=9, Comparison Type=::NewEnumerator2)
      Bool2=&SetBoolByIntComparison(A=$Particles.PartitionParticles.Partition, B=10, Comparison Type=::NewEnumerator7)
     )
    )) {module=KillParticles001}
   &ParticleState()
   /SetVariables(Particles.GlobalSize:NiagaraFloat=&FloatFromCurve(
     FloatCurve=DataInterface() {propertyValues {Curve {keys [{interpMode=RCIM_Cubic, tangentMode=RCTM_User, time=0, value=1}, {interpMode=RCIM_Cubic, tangentMode=RCTM_User, time=1, value=0.2}]}}}
     Scale Curve=&RandomRangeFloat(Minimum=0.4, Maximum=0.6)
    )) {isSetParameters=true}
   &ScaleSpriteSize(
    Uniform Curve Scale=&RandomRangeFloat(Minimum=1, Maximum=1.2)
    Uniform Curve Sprite Scale=DataInterface() {propertyValues {Curve {keys [<at 0 1 RCIM_Cubic>,<at 1 0.2 RCIM_Cubic>]}}}
   )
   &ScaleSpriteSize(
    Scale Sprite Size Mode=::NewEnumerator0
    Uniform Scale Factor=&MakeCustomFloatFromBool(
     Bool=$User.bShowBackFX
     True Float=&SelectFloatFromArray(
      Direct Array Index=$Particles.PartitionParticles.Partition
      Float Selection Array=DataInterface() {propertyValues {FloatData [1,1,1,1,1,1,1]}}
     )
     False Float=&SelectFloatFromArray(
      Direct Array Index=$Particles.PartitionParticles.Partition
      Float Selection Array=DataInterface() {propertyValues {FloatData [1,1,1,1,0,0,1]}}
     )
    )
   ) {module=ScaleSpriteSize002}
   &ScaleSpriteSize(
    Scale Sprite Size Mode=::NewEnumerator0
    Uniform Scale Factor=&MakeCustomFloatFromBool(
     Bool=$User.bShowEyeFX
     True Float=&SelectFloatFromArray(
      Direct Array Index=$Particles.PartitionParticles.Partition
      Float Selection Array=DataInterface() {propertyValues {FloatData [1,1,1,1,1,1,1,1,1,0,0,1]}}
     )
     False Float=&SelectFloatFromArray(
      Direct Array Index=$Particles.PartitionParticles.Partition
      Float Selection Array=DataInterface() {propertyValues {FloatData [1,1,1,1,1,1,1,1,1,0,0,1]}}
     )
    )
   ) {module=ScaleSpriteSize001}
   &CurlNoiseForce(
    Noise Frequency=15
    Noise Strength=&FloatFromCurve(
     FloatCurve=DataInterface() {propertyValues {Curve {keys [<at 0 1 RCIM_Cubic>,<at 0.2 1.875 RCIM_Cubic>,<at 1 1 RCIM_Cubic>]}}}
     CurveIndex=$System.Timeline.Time
     Scale Curve=70
    )
    Pan Noise=true
    Pan Noise Field=<0 0.2 0.05>
   )
   &InheritSourceMovement(
    Applied Force Scale=&VectorFromFloat(Value=&FloatFromCurve(FloatCurve=DataInterface() {propertyValues {Curve {keys [<at 0.6 0 RCIM_Cubic>,<at 1 -0.5>]}}}))
    Applied Position Scale=&VectorFromFloat(Value=&FloatFromCurve(FloatCurve=DataInterface() {propertyValues {Curve {keys [<at 0.1 1 RCIM_Cubic>,<at 0.6 0>]}}}))
    ApplyForce=false
    ApplyPosition=true
    ApplyPositionFalloff=false
   )
   &SolveForcesAndVelocity(Clamp Velocity=false, Limit Acceleration=false)
   &ScaleColor(
    Scale Alpha=&FloatFromCurve(
     FloatCurve=DataInterface() {propertyValues {Curve {keys [
         {interpMode=RCIM_Cubic, tangentMode=RCTM_User, time=0, value=0}
         {interpMode=RCIM_Cubic, tangentMode=RCTM_User, time=0.5, value=1}
         {tangentMode=RCTM_User, time=1, value=0}
        ]}}}
     Scale Curve=&MakeCustomFloatFromBool(Bool=$User.bIsFrontEndPreview, True Float=0.14, False Float=0.14)
    )
    Scale RGB=&VectorFromFloat(Value=&FloatFromCurve(
      FloatCurve=DataInterface() {propertyValues {Curve {keys [<at 0 1 RCIM_Cubic>,<at 0.2 2 RCIM_Cubic>,<at 1 1 RCIM_Cubic>]}}}
      CurveIndex=$System.Timeline.Time
      Scale Curve=$System.TimeOfDayBrightnessMultiplier
     ))
    ScaleA=true
    ScaleRGB=true
   )
  ]
  rendererClasses [&NiagaraSpriteRendererProperties]
  renderers [{rendererClass=&NiagaraSpriteRendererProperties}]
 }
 FireMain {
  emitterUpdateScript [
   &EmitterState(
    Inactive Response=::NewEnumerator1
    Life Cycle Mode=::NewEnumerator1
    Loop Behavior=::NewEnumerator1
    Loop Duration Mode=::NewEnumerator1
    UseLoopDelay=false
   )
   &SpawnRate(SpawnRate=14, Use Spawn Probability=false)
   &SpawnBurst_Instantaneous(Spawn Count=12, Use Spawn Probability=false, UseLoopCountLimit=false)
  ]
  particleSpawnScript [
   &PartitionParticles(Number Of Partitions=12)
   &InitializeParticle(
    Color=<rgba 0.02 0.0003 0 0.6>
    Lifetime Max=1.5
    Lifetime Mode=::NewEnumerator1
    Mesh Scale=&SelectVectorFromArray(
     Direct Array Index=$Particles.PartitionParticles.Partition
     Vector Selection Array=DataInterface() {propertyValues {FloatData [
        <0.75 0.75 0.85>
        <0.75 0.75 0.85>
        <0.75 0.75 0.5>
        <0.75 0.75 0.5>
        <1 1 1>
        <1 1 1>
        <0.65 0.65 0.65>
        <0.65 0.65 0.65>
        <0 0 0>
        <0.4 0.4 0.4>
        <0.4 0.4 0.4>
        <0.7 0.7 0.7>
       ]}}
    )
    Mesh Scale Mode=::NewEnumerator4
    UsePositionOffset=true
    Write Material Random=true
   )
   &InitialMeshOrientation(
    Mesh Orientation Mode=::NewEnumerator2
    Rotation=&SelectVectorFromArray(
     Direct Array Index=$Particles.PartitionParticles.Partition
     Vector Selection Array=DataInterface() {propertyValues {FloatData [
        <0.0129 0.0901 0>
        <-0.0261 0.3719 0>
        <-0.0232 0.0631 0>
        <0.037 0.3809 0>
        <0.005 -0.0541 0>
        <0.005 0.02 0>
        <0 -0.1441 0>
        <0 -0.4054 0>
        <0.1892 0.1982 0.1261>
        <0 0.0176 0>
        <0 0.4586 0>
        <-0.3221 0 -0.2432>
       ]}}
    )
    Use Rotation Vector=true
   )
   /SetVariables(Particles.CameraOffset002:NiagaraFloat=&SelectFloatFromArray(
     Direct Array Index=$Particles.PartitionParticles.Partition
     Float Selection Array=DataInterface() {propertyValues {FloatData [3,3,1,1,0,0,2,2,9,3,3,0]}}
    )) {isSetParameters=true}
   /SetVariables(Particles.CameraOffset003:NiagaraFloat=&SelectFloatFromArray(
     Direct Array Index=$Particles.PartitionParticles.Partition
     Float Selection Array=DataInterface() {propertyValues {FloatData [3,3,1,1,0,0,2,2,9,3,3,0]}}
    )) {isSetParameters=true}
   &DynamicMaterialParameters(
    Index 0 Param 1=&RandomRangeFloat(Minimum=0, Maximum=1)
    Index 0 Param 2=&RandomRangeFloat(Minimum=0, Maximum=1)
    Index 0 Param 3=&SelectFloatFromArray(
     Direct Array Index=$Particles.PartitionParticles.Partition
     Float Selection Array=DataInterface() {propertyValues {FloatData [0.7,0.7,0.7,0.7,0.7,0.7,0.7,0.7,0.7,0.1,0.1,0.7]}}
    )
    Index 0 Param 4=&MakeCustomFloatFromBool(Bool=$User.bIsFrontEndPreview, True Float=0, False Float=1)
    Module.Index0Param0Write=true
    Module.Index0Param1Write=true
    Module.Index0Param2Write=true
    Module.Index0Param3Write=true
   )
  ]
  particleUpdateScript [
   &ParticleState()
   /SetVariables(Particles.Initial.Scale:Vector3f=&MakeCustomVectorFromBool(
     Bool=&Boolean_AndOperation(
      Bool1=$User.bIsHandHidden
      Bool2=&Boolean_AndOperation(
       Bool1=&SetBoolByIntComparison(A=$Particles.PartitionParticles.Partition, B=6, Comparison Type=::NewEnumerator2)
       Bool2=&SetBoolByIntComparison(A=$Particles.PartitionParticles.Partition, B=8, Comparison Type=::NewEnumerator7)
      )
     )
     True Vector=<0 0 0>
     False Vector=$Particles.Initial.Scale
    )) {isSetParameters=true}
   /SetVariables(Particles.Initial.Scale:Vector3f=&MakeCustomVectorFromBool(
     Bool=&Boolean_AndOperation(
      Bool1=$User.bisHeadHidden
      Bool2=&Boolean_AndOperation(
       Bool1=&SetBoolByIntComparison(A=$Particles.PartitionParticles.Partition, B=9, Comparison Type=::NewEnumerator2)
       Bool2=&SetBoolByIntComparison(A=$Particles.PartitionParticles.Partition, B=10, Comparison Type=::NewEnumerator7)
      )
     )
     True Vector=<0 0 0>
     False Vector=$Particles.Initial.Scale
    )) {isSetParameters=true}
   &ScaleMeshSize(Scale Factor=&VectorFromFloat(Value=&MakeCustomFloatFromBool(
      Bool=$User.bShowEyeFX
      True Float=&SelectFloatFromArray(
       Direct Array Index=$Particles.PartitionParticles.Partition
       Float Selection Array=DataInterface() {propertyValues {FloatData [1,1,1,1,1,1,1,1,1,1]}}
      )
      False Float=&SelectFloatFromArray(
       Direct Array Index=$Particles.PartitionParticles.Partition
       Float Selection Array=DataInterface() {propertyValues {FloatData [1,1,1,1,1,1,1,1,1,0,0,1]}}
      )
     ))) {module=ScaleMeshSize001}
   &ScaleMeshSize(Scale Factor=&VectorFromFloat(Value=&MakeCustomFloatFromBool(
      Bool=$User.bShowBackFX
      True Float=&SelectFloatFromArray(
       Direct Array Index=$Particles.PartitionParticles.Partition
       Float Selection Array=DataInterface() {propertyValues {FloatData [1,1,1,1,1,1,1,1,1,1]}}
      )
      False Float=&SelectFloatFromArray(
       Direct Array Index=$Particles.PartitionParticles.Partition
       Float Selection Array=DataInterface() {propertyValues {FloatData [1,1,1,1,0,0,1,1,1,1]}}
      )
     ))) {module=ScaleMeshSize003}
   /SetVariables(Particles.Color001:LinearColor=&MultiplyLinearColor(
     Linear Color=&ColorFromCurve(ColorCurve=DataInterface() {propertyValues {
        AlphaCurve {keys [<at 0 0 RCIM_Cubic>,<at 0.1396 0.8>,<at 0.7438 0.6>,<at 1 0 RCIM_Cubic>]}
        BlueCurve {keys [<at 0 0 RCIM_Cubic>,<at 1 0.0109 RCIM_Cubic>]}
        GreenCurve {keys [<at 0 0.1507 RCIM_Cubic>,<at 1 0.0664 RCIM_Cubic>]}
        RedCurve {keys [<at 0 0.3 RCIM_Cubic>,<at 1 0.3 RCIM_Cubic>]}
       }})
     Scale Factor=&MakeVector4_FromFloat(Float=$System.TimeOfDayBrightnessMultiplier)
    )) {isSetParameters=true}
   /SetVariables(Particles.Color002=&MultiplyLinearColor(
     Linear Color=&ColorFromCurve(ColorCurve=DataInterface() {propertyValues {
        AlphaCurve {keys [<at 0 0>, <at 0.1364 0.6>, {arriveTangent=-0.6947, interpMode=RCIM_Cubic, leaveTangent=-0.6947, time=0.7422, value=0.3}, <at 1 0 RCIM_Cubic>]}
        BlueCurve {keys [<at 0 0 RCIM_Cubic>,<at 0.5642 0 RCIM_Cubic>]}
        GreenCurve {keys [<at 0 3.8291 RCIM_Cubic>,<at 0.5642 2.25 RCIM_Cubic>]}
        RedCurve {keys [<at 0 5 RCIM_Cubic>,<at 0.5642 3 RCIM_Cubic>]}
       }})
     Scale Factor=&MakeVector4_FromFloat(Float=$System.TimeOfDayBrightnessMultiplier)
    )) {isSetParameters=true}
   &ScaleMeshSize(Scale Factor=&VectorFromCurve(
     VectorCurve=DataInterface() {propertyValues {
       XCurve {keys [<at 0 0.625 RCIM_Cubic>, {arriveTangent=-0.0638, interpMode=RCIM_Cubic, leaveTangent=-0.0638, tangentMode=RCTM_User, time=1, value=1.1}]}
       YCurve {keys [<at 0 0.625 RCIM_Cubic>, {arriveTangent=-0.0638, interpMode=RCIM_Cubic, leaveTangent=-0.0638, tangentMode=RCTM_User, time=1, value=1.1}]}
       ZCurve {keys [<at 0 0.625 RCIM_Cubic>, {arriveTangent=-0.0638, interpMode=RCIM_Cubic, leaveTangent=-0.0638, tangentMode=RCTM_User, time=1, value=1.1}]}
      }}
     Scale Curve=<0.75 0.75 0.75>
    ))
   &SkeletalMeshLocation(
    Bone / Socket Index=$Particles.PartitionParticles.Partition
    Bone Sampling Mode=::NewEnumerator4
    MaskSpawnGroup=false
    Sampled Position Offset=&SelectVectorFromArray(
     Direct Array Index=$Particles.PartitionParticles.Partition
     Vector Selection Array=DataInterface() {propertyValues {FloatData [
        <4 12 9>
        <4 12 -9>
        <-7.5 13 7>
        <-7.5 13 -7>
        <7.5 0 -1>
        <-7.5 0 1>
        <5.5 2.5 1.5>
        <5.5 2.5 -1.5>
        <-20 0 0>
        <4 8.6 2>
        <4.3 8.6 -2>
        <6.5 16 2>
       ]}}
    )
    Skeletal Mesh=$System.Skeletal Mesh
   ) {enabled=false}
   &SocketLocation(
    SamplePosOffset=true
    Sampled Position Offset=&SelectVectorFromArray(
     Direct Array Index=$Particles.PartitionParticles.Partition
     Vector Selection Array=DataInterface() {propertyValues {FloatData [
        <4 12 9>
        <4 12 -9>
        <-7.5 13 7>
        <-7.5 13 -7>
        <7.5 0 -1>
        <-7.5 0 1>
        <5.5 2.5 1.5>
        <5.5 2.5 -1.5>
        <-20 0 0>
        <4 8.6 2>
        <4.3 8.6 -2>
        <6.5 16 2>
       ]}}
    )
    Socket Index=$Particles.PartitionParticles.Partition
    Socket Reader=$System.Socket Reader
    Socket Sampling Mode=::NewEnumerator3
    VelocitySpeedLimit=false
   )
   &InterpolateOverTime(Rate Of Change=8, Target Vector Value:NiagaraPosition=$Particles.Position, Variable Type=::NewEnumerator2)
   /SetVariables(Particles.Interpolate Dynamic=&MakeFloat4FromVectorAndFloat("Vector (RGB)"=$Particles.InterpolateOverTime.InterpolateOverTime_V3_.Moving Average, "Float (Alpha)"=0.2)) {isSetParameters=true}
   &ScaleMeshSize(Scale Factor=&VectorFromFloat(Value=&FloatFromCurve(
      FloatCurve=DataInterface() {propertyValues {Curve {keys [
          <at 0 1 RCIM_Cubic>
          {arriveTangent=0.2347, interpMode=RCIM_Cubic, leaveTangent=0.2347, tangentMode=RCTM_User, time=0.2, value=1.55}
          {arriveTangent=-1.3101, interpMode=RCIM_Cubic, leaveTangent=-1.3101, tangentMode=RCTM_User, time=1, value=1}
         ]}}}
      CurveIndex=$System.Timeline.Time
     ))) {module=ScaleMeshSize002}
   &ScaleMeshSize(Scale Factor=&VectorFromFloat(Value=&MakeCustomFloatFromBool(
      Bool=$User.bisADS
      True Float=&SelectFloatFromArray(
       Direct Array Index=$Particles.PartitionParticles.Partition
       Float Selection Array=DataInterface() {propertyValues {FloatData [0,0,0,0,1,1,0,0,0,1,0,0]}}
      )
      False Float=1
     ))) {module=ScaleMeshSize004}
   &ScaleMeshSize(Scale Factor=&VectorFromFloat(Value=&MakeCustomFloatFromBool(
      Bool=$User.bIsADSOther
      True Float=&SelectFloatFromArray(
       Direct Array Index=$Particles.PartitionParticles.Partition
       Float Selection Array=DataInterface() {propertyValues {FloatData [1,1,1,1,1,1,1,1,1,1,0,1]}}
      )
      False Float=1
     ))) {module=ScaleMeshSize005}
  ]
  rendererClasses [&NiagaraMeshRendererProperties]
  renderers [
   {rendererClass=&NiagaraMeshRendererProperties, rendererIndex=0}
   {rendererClass=&NiagaraMeshRendererProperties, rendererIndex=1}
  ]
 }
 FireParticle {
  emitterUpdateScript [
   &EmitterState(
    Enable Distance Culling=true
    Inactive Response=::NewEnumerator1
    Life Cycle Mode=::NewEnumerator1
    Loop Duration=1
    Module.bApplySpawnCountScaleByDistance=true
    Module.bUseMaxDistance=false
    Module.bUseMinDistance=false
    Scalability Mode=::NewEnumerator1
    Scale Spawn Count=true
    Spawn Count Scale By Distance=DataInterface() {propertyValues {Curve {keys [<at 0 1>,<at 0.5 1>,<at 1 0>]}}}
    UseLoopDelay=false
   )
   &SpawnRate(
    SpawnRate=&FloatFromCurve(FloatCurve=DataInterface() {propertyValues {Curve {keys [<at 0 1 RCIM_Cubic>,<at 1 0>]}}}, CurveIndex=$System.Timeline.Time, Scale Curve=25)
    Use Spawn Probability=false
   ) {module=SpawnRate001}
   &SpawnRate(SpawnRate=53, Use Spawn Probability=false)
  ]
  particleSpawnScript [
   &PartitionParticles(Number Of Partitions=12)
   &InitializeParticle(
    Color=<rgba 1 0.8504 0 1.5>
    Lifetime Min=1.6
    Lifetime Mode=::NewEnumerator1
    Sprite Rotation Mode=::NewEnumerator2
    Sprite Size Max=<25 25>
    Sprite Size Min=<15 15>
    Sprite Size Mode=::NewEnumerator2
    UsePositionOffset=false
    Write Material Random=true
   )
   &CameraOffset(Camera Offset Amount=&FN_PickFloatDependOnPlatform(TargetPlatformActive=2, TargetPlatformNoActive=2))
   &DynamicMaterialParameters(
    Index 0 Param 4=&MakeCustomFloatFromBool(Bool=$User.bIsFrontEndPreview, True Float=0, False Float=1)
    Module.Index0Param0Write=false
    Module.Index0Param1Write=false
    Module.Index0Param2Write=false
    Module.Index0Param3Write=true
   )
  ]
  particleUpdateScript [
   &ParticleState()
   &SpriteRotationRate(Rotation Rate=&RandomRangeFloat(Minimum=-155, Maximum=155))
   &ScaleSpriteSize(
    Scale Sprite Size Mode=::NewEnumerator0
    Uniform Scale Factor=&MakeCustomFloatFromBool(
     Bool=$User.bShowBackFX
     True Float=&SelectFloatFromArray(
      Direct Array Index=$Particles.PartitionParticles.Partition
      Float Selection Array=DataInterface() {propertyValues {FloatData [1,1,1,1,1,1,1]}}
     )
     False Float=&SelectFloatFromArray(
      Direct Array Index=$Particles.PartitionParticles.Partition
      Float Selection Array=DataInterface() {propertyValues {FloatData [1,1,1,1,0,0,1]}}
     )
    )
   ) {module=ScaleSpriteSize002}
   &ScaleSpriteSize(
    Scale Sprite Size Mode=::NewEnumerator0
    Uniform Scale Factor=&MakeCustomFloatFromBool(
     Bool=$User.bShowEyeFX
     True Float=&SelectFloatFromArray(
      Direct Array Index=$Particles.PartitionParticles.Partition
      Float Selection Array=DataInterface() {propertyValues {FloatData [0.6,0.6,0.6,0.6,0.65,0.65,0,0,0,0.3,0.3,0.7]}}
     )
     False Float=&SelectFloatFromArray(
      Direct Array Index=$Particles.PartitionParticles.Partition
      Float Selection Array=DataInterface() {propertyValues {FloatData [0.6,0.6,0.6,0.6,0.65,0.65,0,0,0,0,0,0.7]}}
     )
    )
   ) {module=ScaleSpriteSize001}
   &SkeletalMeshLocation(
    Bone / Socket Index=$Particles.PartitionParticles.Partition
    Bone Sampling Mode=::NewEnumerator4
    MaskSpawnGroup=false
    Sampled Position Offset=&SelectVectorFromArray(
     Direct Array Index=$Particles.PartitionParticles.Partition
     Vector Selection Array=DataInterface() {propertyValues {FloatData [
        <4 12 9>
        <4 12 -9>
        <-7.5 13 7>
        <-7.5 13 -7>
        <6.5 1 -1>
        <-6.5 1 1>
        <5.5 2 1.5>
        <5.5 2 -1.5>
        <-22 0 0>
        <3.6 9 1.5>
        <3.6 9 -1.5>
        <7 14 1>
       ]}}
    )
    Skeletal Mesh=$System.Skeletal Mesh
   ) {enabled=false}
   &SocketLocation(
    SamplePosOffset=true
    Sampled Position Offset=&SelectVectorFromArray(
     Direct Array Index=$Particles.PartitionParticles.Partition
     Vector Selection Array=DataInterface() {propertyValues {FloatData [
        <4 12 9>
        <4 12 -9>
        <-7.5 13 7>
        <-7.5 13 -7>
        <6.5 1 -1>
        <-6.5 1 1>
        <5.5 2 1.5>
        <5.5 2 -1.5>
        <-22 0 0>
        <3.6 9 1.5>
        <3.6 9 -1.5>
        <7 14 1>
       ]}}
    )
    Socket Index=$Particles.PartitionParticles.Partition
    Socket Reader=$System.Socket Reader
    Socket Sampling Mode=::NewEnumerator3
    VelocitySpeedLimit=false
   )
   /SetVariables(Particles.SubImageIndex=&FloatFromCurve(FloatCurve=DataInterface() {propertyValues {Curve {keys [<at 0 5>,<at 1 15>]}}})) {isSetParameters=true}
   &ScaleSpriteSize(Scale Factor=&Vector2DFromFloat(Value=&FloatFromCurve(
      FloatCurve=DataInterface() {propertyValues {Curve {keys [
          {interpMode=RCIM_Cubic, leaveTangent=3.0564, tangentMode=RCTM_Break, time=0, value=0}
          {arriveTangent=1.1454, interpMode=RCIM_Cubic, leaveTangent=1.1454, tangentMode=RCTM_User, time=0.05, value=0.8}
          {arriveTangent=-1.5127, interpMode=RCIM_Cubic, tangentMode=RCTM_Break, time=1, value=0.6}
         ]}}}
      Scale Curve=0.8
     )))
   &ScaleColor(
    Scale Alpha=&FloatFromCurve(
     FloatCurve=DataInterface() {propertyValues {Curve {keys [
         <at 0 0 RCIM_Cubic>
         <at 0.2 1 RCIM_Cubic>
         {arriveTangent=-1.4238, interpMode=RCIM_Cubic, leaveTangent=-1.4238, tangentMode=RCTM_User, time=1, value=0}
        ]}}}
     Scale Curve=0.09
    )
    Scale RGB=&MakeVectorFromLinearColorRGB(Color=&ColorFromCurve(ColorCurve=DataInterface() {propertyValues {
        AlphaCurve {keys [<at 0 1 RCIM_Cubic>]}
        BlueCurve {keys [<at 0.1048 0 RCIM_Cubic>,<at 0.5156 0 RCIM_Cubic>]}
        GreenCurve {keys [<at 0.1048 0.7758 RCIM_Cubic>,<at 0.5156 0.1401 RCIM_Cubic>]}
        RedCurve {keys [<at 0.1048 1 RCIM_Cubic>,<at 0.5156 1 RCIM_Cubic>]}
       }}))
    ScaleA=true
    ScaleRGB=true
   )
   /SetVariables(Particles.Position=&AddVectorToPosition(
     Position=$Particles.Position
     Vector=&Lerp_Vector(
      A=<0 0 0>
      B=&Add_Vector(
       A=&Multiply_Vector(
        A=&MakeVector(
         X=&SelectFloatFromArray(
          Direct Array Index=$Particles.PartitionParticles.Partition
          Float Selection Array=DataInterface() {propertyValues {FloatData [-1.5,1.5,-1.4,1.4,0,0,3,-3,0,-0.5,0.5,0]}}
         )
         Y=&SelectFloatFromArray(
          Direct Array Index=$Particles.PartitionParticles.Partition
          Float Selection Array=DataInterface() {propertyValues {FloatData [0,0,0,0,0,0,0,0,-1,0,0,0.6]}}
         )
         Z=&SelectFloatFromArray(
          Direct Array Index=$Particles.PartitionParticles.Partition
          Float Selection Array=DataInterface() {propertyValues {FloatData [0.9,0.9,0.7,0.7,2.2,2.2,2,2,0,0.2,0.2,0.6]}}
         )
        )
        B=&RandomRangeVector(Minimum=<15 15 15>, Maximum=<20 20 20>)
       )
       B=&Multiply_Vector(
        A=&NormalizeVector(Vector To Normalize=$Engine.Owner.Velocity, Safe Normalize=true, Safe Fallback Threshold=0, Safe Fallback Vector=<0 0 0.001>)
        B=&VectorFromFloat(Value=&FloatFromCurve(
          FloatCurve=DataInterface() {propertyValues {Curve {keys [{interpMode=RCIM_Cubic, tangentMode=RCTM_User, time=0, value=0}, <at 1 1>]}}}
          Scale Curve=2
         ))
       )
      )
      Alpha=&FloatFromCurve(
       FloatCurve=DataInterface() {propertyValues {Curve {keys [
           {arriveTangent=0.3072, interpMode=RCIM_Cubic, leaveTangent=0.3072, tangentMode=RCTM_User, time=0, value=0}
           {arriveTangent=1.042, tangentMode=RCTM_User, time=1, value=1}
          ]}}}
       Scale Curve=&FloatFromCurve(
        FloatCurve=DataInterface() {propertyValues {Curve {keys [<at 0 1 RCIM_Cubic>,<at 0.2 1.925 RCIM_Cubic>,<at 1 1 RCIM_Cubic>]}}}
        CurveIndex=$System.Timeline.Time
       )
      )
     )
    )) {isSetParameters=true}
   &ScaleColor(
    Scale RGB=&VectorFromFloat(Value=&FloatFromCurve(
      FloatCurve=DataInterface() {propertyValues {Curve {keys [<at 0 1 RCIM_Cubic>,<at 0.2 1.35 RCIM_Cubic>,<at 1 1 RCIM_Cubic>]}}}
      CurveIndex=$System.Timeline.Time
      Scale Curve=$System.TimeOfDayBrightnessMultiplier
     ))
    ScaleA=true
    ScaleRGB=true
   ) {module=ScaleColor001}
   &ScaleSpriteSize(
    Scale Sprite Size Mode=::NewEnumerator0
    Uniform Scale Factor=&MakeCustomFloatFromBool(
     Bool=$User.bisADS
     True Float=&SelectFloatFromArray(
      Direct Array Index=$Particles.PartitionParticles.Partition
      Float Selection Array=DataInterface() {propertyValues {FloatData [0,0,0,0,1,1,0,0,0,1,0,0]}}
     )
     False Float=1
    )
   ) {module=ScaleSpriteSize003}
   &ScaleSpriteSize(
    Scale Sprite Size Mode=::NewEnumerator0
    Uniform Scale Factor=&MakeCustomFloatFromBool(
     Bool=$User.bIsADSOther
     True Float=&SelectFloatFromArray(
      Direct Array Index=$Particles.PartitionParticles.Partition
      Float Selection Array=DataInterface() {propertyValues {FloatData [1,1,1,1,1,1,1,1,1,1,0,1]}}
     )
     False Float=1
    )
   ) {module=ScaleSpriteSize004}
   /SetVariables(Particles.VisibilityTag=&MakeCustomIntFromBool(
     Bool=$User.bisHeadHidden
     True INT=&MakeCustomIntFromBool(
      Bool=&Boolean_OrOperation(
       Bool1=&SetBoolByIntComparison(A=$Particles.PartitionParticles.Partition, B=9, Comparison Type=::NewEnumerator4)
       Bool2=&SetBoolByIntComparison(A=$Particles.PartitionParticles.Partition, B=10, Comparison Type=::NewEnumerator4)
      )
      True INT=1
      False INT=0
     )
     False INT=0
    )) {isSetParameters=true}
  ]
  rendererClasses [&NiagaraSpriteRendererProperties]
  renderers [{rendererClass=&NiagaraSpriteRendererProperties}]
 }
 Glow {
  emitterUpdateScript [
   &EmitterState(
    Inactive Response=::NewEnumerator1
    Life Cycle Mode=::NewEnumerator1
    Loop Behavior=::NewEnumerator1
    Loop Duration Mode=::NewEnumerator1
    UseLoopDelay=false
   )
   &SpawnBurst_Instantaneous(Spawn Count=10, Use Spawn Probability=false, UseLoopCountLimit=false)
  ]
  particleSpawnScript [
   &PartitionParticles(Number Of Partitions=10)
   &InitializeParticle(
    Color=<rgba 1 0.6872 0 0.055>
    Lifetime=5
    Sprite Size Mode=::NewEnumerator3
    Uniform Sprite Size=&SelectFloatFromArray(
     Direct Array Index=$Particles.PartitionParticles.Partition
     Float Selection Array=DataInterface() {propertyValues {FloatData [30,30,25,25,25,25,20,15,35,35]}}
    )
    UsePositionOffset=false
    Write Material Random=true
   )
   &CameraOffset(Camera Offset Amount=&SelectFloatFromArray(
     Direct Array Index=$Particles.PartitionParticles.Partition
     Float Selection Array=DataInterface() {propertyValues {FloatData [12,12,12,12,12,12,4,12,6,6]}}
    ))
  ]
  particleUpdateScript [
   &ParticleState(Kill Particles When Lifetime Has Elapsed=false, Let Infinitely Lived Particles Die When Emitter Deactivates=true)
   /SetVariables(
    Particles.BeltBones:NiagaraInt32=&SelectIntFromArray(
     Direct Array Index=$Particles.PartitionParticles.Partition
     Int Selection Array=DataInterface() {propertyValues {IntData [0,0,0,0,0,0,0,0,1,1]}}
    )
    Particles.DefaultBones:NiagaraInt32=&MakeCustomIntFromBool(
     Bool=$User.bisHeadHidden
     True INT=&SelectIntFromArray(
      Direct Array Index=$Particles.PartitionParticles.Partition
      Int Selection Array=DataInterface() {propertyValues {IntData [0,0,0,0,0,0,0,1,1,1]}}
     )
     False INT=&SelectIntFromArray(
      Direct Array Index=$Particles.PartitionParticles.Partition
      Int Selection Array=DataInterface() {propertyValues {IntData [0,0,0,0,0,0,0,0,1,1]}}
     )
    )
    Particles.LightsVisibilityTag:NiagaraInt32=&SelectIntFromArray(
     Direct Array Index=$Particles.PartitionParticles.Partition
     Int Selection Array=DataInterface() {propertyValues {IntData [0,0,1,1,1,1,1,1,1,1]}}
    )
   ) {isSetParameters=true}
   &ScaleSpriteSize(
    Scale Sprite Size Mode=::NewEnumerator0
    Uniform Scale Factor=&Multiply_Float(
     A=&MakeCustomFloatFromBool(
      Bool=$User.bShowBackFX
      True Float=&SelectFloatFromArray(
       Direct Array Index=$Particles.PartitionParticles.Partition
       Float Selection Array=DataInterface() {propertyValues {FloatData [1,1,1,1,1,1,1]}}
      )
      False Float=&SelectFloatFromArray(
       Direct Array Index=$Particles.PartitionParticles.Partition
       Float Selection Array=DataInterface() {propertyValues {FloatData [1,1,1,1,0,0,1]}}
      )
     )
     B=&MakeCustomFloatFromBool(
      Bool=&Boolean_AndOperation(
       Bool1=$User.bIsHandHidden
       Bool2=&SetBoolByIntComparison(A=$Particles.PartitionParticles.Partition, B=6, Comparison Type=::NewEnumerator4)
      )
      True Float=0
      False Float=&MakeCustomFloatFromBool(
       Bool=&Boolean_AndOperation(
        Bool1=$System.LODVisibility
        Bool2=&Boolean_OrOperation(
         Bool1=&SetBoolByIntComparison(A=$Particles.PartitionParticles.Partition, B=8, Comparison Type=::NewEnumerator4)
         Bool2=&SetBoolByIntComparison(A=$Particles.PartitionParticles.Partition, B=9, Comparison Type=::NewEnumerator4)
        )
       )
       True Float=0
       False Float=1
      )
     )
    )
   )
   &ScaleColor(
    Scale Alpha=&FloatFromCurve(
     FloatCurve=DataInterface() {propertyValues {Curve {keys [
         {interpMode=RCIM_Cubic, tangentMode=RCTM_User, time=0, value=0.6}
         {arriveTangent=0.0833, interpMode=RCIM_Cubic, leaveTangent=0.0833, time=0.1574, value=0.7281}
         {arriveTangent=-0.3221, interpMode=RCIM_Cubic, leaveTangent=-0.3221, time=0.3, value=0.625}
         {arriveTangent=1.0245, interpMode=RCIM_Cubic, leaveTangent=1.0245, time=0.4, value=0.65}
         {arriveTangent=1.1515, interpMode=RCIM_Cubic, leaveTangent=1.1515, time=0.4386, value=0.767}
         {interpMode=RCIM_Cubic, tangentMode=RCTM_User, time=0.5185, value=0.7864}
         {arriveTangent=-0.469, interpMode=RCIM_Cubic, leaveTangent=-0.469, time=0.6286, value=0.578}
         {arriveTangent=0.0711, interpMode=RCIM_Cubic, leaveTangent=0.0711, time=0.7672, value=0.6698}
         {arriveTangent=-0.5596, interpMode=RCIM_Cubic, leaveTangent=-0.5596, time=0.8293, value=0.5922}
         {arriveTangent=-0.3306, interpMode=RCIM_Cubic, leaveTangent=-0.3306, time=0.9058, value=0.5922}
         {arriveTangent=0.0826, interpMode=RCIM_Cubic, leaveTangent=0.0826, time=0.9467, value=0.5534}
         {tangentMode=RCTM_User, time=1, value=0.6}
        ]}}}
     CurveIndex=$Emitter.LoopedAge
     Scale Curve=&RandomRangeFloat(Minimum=1.8, Maximum=2.2)
    )
    Scale RGB=&VectorFromFloat(Value=&FloatFromCurve(
      FloatCurve=DataInterface() {propertyValues {Curve {keys [<at 0 1 RCIM_Cubic>,<at 0.2 2 RCIM_Cubic>,<at 1 1 RCIM_Cubic>]}}}
      CurveIndex=$System.Timeline.Time
      Scale Curve=$System.TimeOfDayBrightnessMultiplier
     ))
    ScaleA=true
    ScaleRGB=true
   )
   &SocketLocation(
    SamplePosOffset=true
    Sampled Position Offset=&SelectVectorFromArray(
     Direct Array Index=$Particles.PartitionParticles.Partition
     Vector Selection Array=DataInterface() {propertyValues {FloatData [<4 12 9>, <4 12 -9>, <-6 10 6>, <-6 11 -6>, <7 0 6>, <-7 0 6>, <5.5 5 1.5>, <4 8.6 0>, <0 0 0>, <0 0 0>]}}
    )
    Socket Index=&SelectIntFromArray(
     Direct Array Index=$Particles.PartitionParticles.Partition
     Int Selection Array=DataInterface() {propertyValues {IntData [0,1,2,3,4,5,6,9,12,13]}}
    )
    Socket Reader=$System.Socket Reader
    Socket Sampling Mode=::NewEnumerator3
    VelocitySpeedLimit=false
   )
   &FN_Light_Attributes(
    Color=<rgba 1 0.647 0 1>
    Create Color Attribute=true
    Create Exponent Attribute=true
    Create Radius Attribute=true
    Enable Particle Light Enable=true
    Light Enabled:NiagaraBool=&Boolean_NotOperation(Bool=&Boolean_AndOperation(
      Bool1=&SetBoolByIntComparison(A=$Particles.PartitionParticles.Partition, B=9, Comparison Type=::NewEnumerator2)
      Bool2=$User.bisHeadHidden
     ))
    Radius=25
    Scale Color=&FloatFromCurve(
     FloatCurve=DataInterface() {propertyValues {Curve {keys [
         {interpMode=RCIM_Cubic, tangentMode=RCTM_User, time=0, value=0.6}
         {arriveTangent=0.3844, interpMode=RCIM_Cubic, leaveTangent=0.3844, time=0.1574, value=0.7281}
         {arriveTangent=-0.361, interpMode=RCIM_Cubic, leaveTangent=-0.361, time=0.2828, value=0.7087}
         {arriveTangent=0.374, interpMode=RCIM_Cubic, leaveTangent=0.374, time=0.3726, value=0.6505}
         {arriveTangent=0.9318, interpMode=RCIM_Cubic, leaveTangent=0.9318, time=0.4386, value=0.767}
         {interpMode=RCIM_Cubic, tangentMode=RCTM_User, time=0.5185, value=0.7864}
         {arriveTangent=-0.5466, interpMode=RCIM_Cubic, leaveTangent=-0.5466, time=0.6286, value=0.6311}
         {arriveTangent=-0.1936, interpMode=RCIM_Cubic, leaveTangent=-0.1936, time=0.7672, value=0.6505}
         {arriveTangent=-0.4203, interpMode=RCIM_Cubic, leaveTangent=-0.4203, time=0.8293, value=0.5922}
         {arriveTangent=-0.3306, interpMode=RCIM_Cubic, leaveTangent=-0.3306, time=0.9058, value=0.5922}
         {arriveTangent=0.0825, interpMode=RCIM_Cubic, leaveTangent=0.0825, time=0.9467, value=0.5534}
         {tangentMode=RCTM_User, time=1, value=0.6}
        ]}}}
     CurveIndex=$Emitter.LoopedAge
     Scale Curve=&Multiply_Float(A=&RandomRangeFloat(Minimum=8, Maximum=12), B=$System.TimeOfDayBrightnessMultiplier)
    )
   )
  ]
  rendererClasses [&NiagaraSpriteRendererProperties, &NiagaraLightRendererProperties]
  renderers [
   {rendererClass=&NiagaraSpriteRendererProperties}
   {rendererClass=&NiagaraLightRendererProperties, rendererIndex=1}
   {rendererClass=&NiagaraSpriteRendererProperties, rendererIndex=2}
  ]
 }
}
summary {
 systemName=NS_CorkFloorSwim_Idle
 emitters [
  {emitterName=Embers, rendererClasses [&NiagaraSpriteRendererProperties]}
  {emitterName=FireMain, rendererClasses [&NiagaraMeshRendererProperties]}
  {emitterName=FireParticle, rendererClasses [&NiagaraSpriteRendererProperties]}
  {emitterName=Glow, rendererClasses [&NiagaraSpriteRendererProperties, &NiagaraLightRendererProperties]}
 ]
 userVariables [
  User.bisADS {type=NiagaraBool}
  User.bIsADSOther {type=NiagaraBool}
  User.bIsFrontEnd {type=NiagaraBool}
  User.bIsFrontEndPreview {type=NiagaraBool}
  User.bIsHandHidden {type=NiagaraBool}
  User.bisHeadHidden {type=NiagaraBool}
  User.bShowBackFX {type=NiagaraBool, defaultValue=true}
  User.bShowEyeFX {type=NiagaraBool, defaultValue=true}
  User.DebugReactivity {type=NiagaraBool}
  User.Empowered {type=NiagaraBool}
  User.FXScale {type=NiagaraFloat, defaultValue=1, description=""}
  User.KillReactivityCount {type=NiagaraInt32}
  User.SkeletalMeshCurrentLOD {type=NiagaraInt32}
 ]
}
