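# Physics asset that simulates the dangling or swinging belt on the CorkFloorSwim medium male player cosmetic body.
# Front and back chains of capsule bodies link with limited swing and twist between left and right sim bodies, forming a closed loop.
# The sim bodies lock angularly to root and right anchor bodies, and each chain's 07 link also locks to the right side.
# Unconstrained pelvis and thigh capsules presumably serve as body colliders.
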PhysicsAsset /BRCosmetics/Characters/Player/Male/Medium/Bodies/M_MED_CorkFloorSwim/Meshes/M_MED_CorkFloorSwim_Belt_Physics1
bodies {
 dyn_belt_04 [{shapeType=Capsule, length=4.7662, radius=4.4338, shapeName=dyn_belt_04_capsule, center=<4.307 0 -4>, rotation=<rot 87.1944 179.9963 179.997>}]
 dyn_belt_05 [{
   shapeType=Capsule
   length=4.7662
   radius=4.4338
   shapeName=dyn_belt_05_capsule
   center=<4.3672 0 -4>
   rotation=<rot 87.1944 179.9965 179.997>
  }]
 dyn_belt_06 [{
   shapeType=Capsule
   length=4.7662
   radius=4.4338
   shapeName=dyn_belt_06_capsule
   center=<4.2665 0 -4>
   rotation=<rot 87.1943 179.997 179.9972>
  }]
 dyn_belt_07 [{
   shapeType=Capsule
   length=4.4841
   radius=4.7159
   shapeName=dyn_belt_07_capsule
   center=<4.1856 0 -4>
   rotation=<rot 87.1943 179.9982 179.998>
  }]
 dyn_belt_08 [{
   shapeType=Capsule
   length=4.3319
   radius=4.8681
   shapeName=dyn_belt_08_capsule
   center=<4.3656 0 -4>
   rotation=<rot 87.1943 179.9998 179.9992>
  }]
 dyn_belt_anchor_r [{shapeType=Capsule, length=0.505, radius=0.505, shapeName=dyn_belt_anchor_r_capsule, rotation=<rot 0 -90 0>}]
 dyn_belt_bk_03 [{shapeType=Capsule, length=4.3319, radius=4.8681, shapeName=dyn_belt_08_capsule, center=<4.2 0 2.4>, rotation=<rot 87.1943 179.9998 179.9992>}]
 dyn_belt_bk_04 [{shapeType=Capsule, length=4.3319, radius=4.8681, shapeName=dyn_belt_08_capsule, center=<4.2 0 2.4>, rotation=<rot 87.1943 179.9998 179.9992>}]
 dyn_belt_bk_05 [{shapeType=Capsule, length=4.3319, radius=4.8681, shapeName=dyn_belt_08_capsule, center=<4.2 0 2.4>, rotation=<rot 87.1943 179.9998 179.9992>}]
 dyn_belt_bk_06 [{shapeType=Capsule, length=4.0498, radius=5.1502, shapeName=dyn_belt_08_capsule, center=<4.2 0 2.4>, rotation=<rot 87.1943 179.9998 179.9992>}]
 dyn_belt_bk_07 [{shapeType=Capsule, length=4.0498, radius=5.1502, shapeName=dyn_belt_08_capsule, center=<4.2 0 2.4>, rotation=<rot 87.1943 179.9998 179.9992>}]
 dyn_belt_bk_08 [{shapeType=Capsule, length=4.0498, radius=5.1502, shapeName=dyn_belt_08_capsule, center=<4.2 0 2.4>, rotation=<rot 87.1943 179.9998 179.9992>}]
 dyn_belt_root [{shapeType=Capsule, length=0.505, radius=0.505, shapeName=dyn_belt_root_capsule, center=<2.0678 0 -0.4154>, rotation=<rot 78.6416 0 0.0001>}]
 dyn_belt_sim_l [{shapeType=Capsule, length=7.5383, radius=4.5353, shapeName=dyn_belt_sim_r_capsule, rotation=<rot 90 0 0>}]
 dyn_belt_sim_r [{shapeType=Capsule, length=7.5383, radius=4.5353, shapeName=dyn_belt_sim_r_capsule, rotation=<rot 90 0 0>}]
 pelvis [{
   shapeType=Capsule
   length=72.2869
   radius=13.4096
   shapeName=pelvis_capsule
   center=<-0.1096 -1.5187 0>
   rotation=<rot 90 179.8321 -179.9752>
  }]
 thigh_l [{
   shapeType=Capsule
   length=40.1189
   radius=10.0458
   shapeName=thigh_l_capsule
   center=<-21.4812 -0.8303 -0.0673>
   rotation=<rot 89.3054 179.9943 179.9948>
  }]
 thigh_r [{
   shapeType=Capsule
   length=39.9188
   radius=10.2579
   shapeName=thigh_r_capsule
   center=<21.7474 1.3337 1.2994>
   rotation=<rot 87.3096 179.9979 179.9985>
  }]
}
constraints [
 {bone1Name=dyn_belt_04, bone2Name=dyn_belt_sim_l, swing1LimitDegrees=40, swing2LimitDegrees=40, twistLimitDegrees=10}
 {bone1Name=dyn_belt_05, bone2Name=dyn_belt_04, swing1LimitDegrees=40, swing2LimitDegrees=40, twistLimitDegrees=10}
 {bone1Name=dyn_belt_06, bone2Name=dyn_belt_05, swing1LimitDegrees=40, swing2LimitDegrees=40, twistLimitDegrees=10}
 {bone1Name=dyn_belt_07, bone2Name=dyn_belt_06, swing1LimitDegrees=40, swing2LimitDegrees=40, twistLimitDegrees=10}
 {bone1Name=dyn_belt_08, bone2Name=dyn_belt_07, swing1LimitDegrees=40, swing2LimitDegrees=40, twistLimitDegrees=10}
 {bone1Name=dyn_belt_bk_04, bone2Name=dyn_belt_bk_03, swing1LimitDegrees=40, swing2LimitDegrees=40, twistLimitDegrees=10}
 {bone1Name=dyn_belt_bk_05, bone2Name=dyn_belt_bk_04, swing1LimitDegrees=40, swing2LimitDegrees=40, twistLimitDegrees=10}
 {bone1Name=dyn_belt_bk_06, bone2Name=dyn_belt_bk_05, swing1LimitDegrees=40, swing2LimitDegrees=40, twistLimitDegrees=10}
 {bone1Name=dyn_belt_bk_07, bone2Name=dyn_belt_bk_06, swing1LimitDegrees=40, swing2LimitDegrees=40, twistLimitDegrees=10}
 {bone1Name=dyn_belt_bk_08, bone2Name=dyn_belt_bk_07, swing1LimitDegrees=40, swing2LimitDegrees=40, twistLimitDegrees=10}
 {
  bone1Name=dyn_belt_sim_r
  bone2Name=dyn_belt_anchor_r
  swing1LimitDegrees=45
  swing1Motion=Locked
  swing2LimitDegrees=45
  swing2Motion=Locked
  twistLimitDegrees=45
  twistMotion=Locked
 }
 {bone1Name=dyn_belt_bk_08, bone2Name=dyn_belt_sim_r, swing1LimitDegrees=40, swing2LimitDegrees=40, twistLimitDegrees=10}
 {bone1Name=dyn_belt_08, bone2Name=dyn_belt_sim_r, swing1LimitDegrees=40, swing2LimitDegrees=40, twistLimitDegrees=10}
 {bone1Name=dyn_belt_bk_03, bone2Name=dyn_belt_sim_l, swing1LimitDegrees=40, swing2LimitDegrees=40, twistLimitDegrees=10}
 {
  bone1Name=dyn_belt_sim_l
  bone2Name=dyn_belt_root
  swing1LimitDegrees=45
  swing1Motion=Locked
  swing2LimitDegrees=45
  swing2Motion=Locked
  twistLimitDegrees=45
  twistMotion=Locked
 }
 {
  bone1Name=dyn_belt_bk_07
  bone2Name=dyn_belt_sim_r
  swing1LimitDegrees=45
  swing1Motion=Locked
  swing2LimitDegrees=45
  swing2Motion=Locked
  twistLimitDegrees=45
  twistMotion=Locked
 }
 {
  bone1Name=dyn_belt_07
  bone2Name=dyn_belt_sim_r
  swing1LimitDegrees=45
  swing1Motion=Locked
  swing2LimitDegrees=45
  swing2Motion=Locked
  twistLimitDegrees=45
  twistMotion=Locked
 }
]
