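# Control Rig for the M_MED_CorkFloorSwim character body that adds corrective deformation and drives its belt, shorts and armor bones.
# After a shared DeformRig function, an RBAN Enabled branch picks the belt solve.
# Weighted parent constraints pin the belt to body bones, and front and back B-spline IK chains shape it.
# Spherical pose readers, mostly smoothed by AlphaInterp, weight transform offsets that move shorts and armor with limb poses.
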ControlRigBlueprint /BRCosmetics/Characters/Player/Male/Medium/Bodies/M_MED_CorkFloorSwim/Meshes/M_MED_CorkFloorSwim_CtrlRig1
graphs [
 New Function_ContainedGraph {nodes [
   RigVMFunctionEntryNode()
   > &RigUnit_ParentConstraint(
    AdvancedSettings=(RotationOrderForFilter=XZY)
    Child=(Type=Bone,Name="dyn_belt_root")
    Parents="((Item=(Type=Bone,Name=\"deform_glute_l\"),Weight=0.0231),(Item=(Type=Bone,Name=\"spine_04\"),Weight=0.0324),(Item=(Type=Bone,Name=\"deform_groin_l\"),Weight=0.0337),(Item=(Type=Bone,Name=\"spine_03\"),Weight=0.0441),(Item=(Type=Bone,Name=\"spine_01\"),Weight=0.0471 ...(truncated)"
    Weight=1
   )
   > &RigUnit_ParentConstraint(
    AdvancedSettings=(RotationOrderForFilter=XZY)
    Child=(Type=Null,Name="dyn_belt_01_null")
    Parents="((Item=(Type=Bone,Name=\"deform_glute_l\"),Weight=0.0231),(Item=(Type=Bone,Name=\"spine_04\"),Weight=0.0324),(Item=(Type=Bone,Name=\"deform_groin_l\"),Weight=0.0337),(Item=(Type=Bone,Name=\"spine_03\"),Weight=0.0441),(Item=(Type=Bone,Name=\"spine_01\"),Weight=0.0471 ...(truncated)"
    Weight=1
   )
   > &RigUnit_ParentConstraint(
    AdvancedSettings=(RotationOrderForFilter=XZY)
    Child=(Type=Null,Name="dyn_belt_02_null")
    Parents="((Item=(Type=Bone,Name=\"deform_glute_l\"),Weight=0.0101),(Item=(Type=Bone,Name=\"spine_03\"),Weight=0.0391),(Item=(Type=Bone,Name=\"spine_01\"),Weight=0.047),(Item=(Type=Bone,Name=\"spine_02\"),Weight=0.0618),(Item=(Type=Bone,Name=\"deform_groin_l\"),Weight=0.0983) ...(truncated)"
    Weight=1
   )
   > &RigUnit_ParentConstraint(
    AdvancedSettings=(RotationOrderForFilter=XZY)
    Child=(Type=Null,Name="dyn_belt_03_null")
    Parents="((Item=(Type=Bone,Name=\"deform_groin_r\"),Weight=0.008),(Item=(Type=Bone,Name=\"thigh_twist_01_r\"),Weight=0.0119),(Item=(Type=Bone,Name=\"spine_03\"),Weight=0.0449),(Item=(Type=Bone,Name=\"spine_01\"),Weight=0.0548),(Item=(Type=Bone,Name=\"spine_02\"),Weight=0.070 ...(truncated)"
    Weight=1
   )
   > &RigUnit_ParentConstraint(
    AdvancedSettings=(RotationOrderForFilter=XZY)
    Child=(Type=Null,Name="dyn_belt_bk_01_null")
    Parents="((Item=(Type=Bone,Name=\"spine_04\"),Weight=0.0123),(Item=(Type=Bone,Name=\"spine_01\"),Weight=0.0154),(Item=(Type=Bone,Name=\"spine_03\"),Weight=0.0157),(Item=(Type=Bone,Name=\"spine_02\"),Weight=0.0212),(Item=(Type=Bone,Name=\"deform_glute_l\"),Weight=0.0219),(Ite ...(truncated)"
    Weight=1
   )
   > &RigUnit_ParentConstraint(
    AdvancedSettings=(RotationOrderForFilter=XZY)
    Child=(Type=Null,Name="dyn_belt_bk_02_null")
    Parents="((Item=(Type=Bone,Name=\"deform_glute_r\"),Weight=0.0082),(Item=(Type=Bone,Name=\"spine_04\"),Weight=0.01),(Item=(Type=Bone,Name=\"spine_01\"),Weight=0.0386),(Item=(Type=Bone,Name=\"spine_03\"),Weight=0.0497),(Item=(Type=Bone,Name=\"spine_02\"),Weight=0.0602),(Item= ...(truncated)"
    Weight=1
   )
   > RigVMFunctionReturnNode()
  ]}
 New Function_ContainedGraph {nodes [
   RigVMFunctionEntryNode()
   > &RigUnit_ParentConstraint(
    AdvancedSettings=(RotationOrderForFilter=XZY)
    Child=(Type=Null,Name="dyn_belt_04_null")
    Parents="((Item=(Type=Bone,Name=\"spine_03\"),Weight=0.0278),(Item=(Type=Bone,Name=\"deform_groin_r\"),Weight=0.0279),(Item=(Type=Bone,Name=\"spine_01\"),Weight=0.0364),(Item=(Type=Bone,Name=\"thigh_twist_01_r\"),Weight=0.0383),(Item=(Type=Bone,Name=\"spine_02\"),Weight=0.04 ...(truncated)"
    Weight=1
   )
   > &RigUnit_ParentConstraint(
    AdvancedSettings=(RotationOrderForFilter=XZY)
    Child=(Type=Null,Name="dyn_belt_05_null")
    Parents="((Item=(Type=Bone,Name=\"spine_03\"),Weight=0.0122),(Item=(Type=Bone,Name=\"spine_01\"),Weight=0.0178),(Item=(Type=Bone,Name=\"spine_02\"),Weight=0.0222),(Item=(Type=Bone,Name=\"deform_groin_r\"),Weight=0.0783),(Item=(Type=Bone,Name=\"deform_groin_l\"),Weight=0.0992 ...(truncated)"
    Weight=1
   )
   > &RigUnit_ParentConstraint(
    AdvancedSettings=(RotationOrderForFilter=XZY)
    Child=(Type=Null,Name="dyn_belt_06_null")
    Parents="((Item=(Type=Bone,Name=\"spine_03\"),Weight=0.0074),(Item=(Type=Bone,Name=\"spine_01\"),Weight=0.0109),(Item=(Type=Bone,Name=\"spine_02\"),Weight=0.0131),(Item=(Type=Bone,Name=\"deform_groin_l\"),Weight=0.0225),(Item=(Type=Bone,Name=\"thigh_twist_01_l\"),Weight=0.03 ...(truncated)"
    Weight=1
   )
   > &RigUnit_ParentConstraint(
    AdvancedSettings=(RotationOrderForFilter=XZY)
    Child=(Type=Null,Name="dyn_belt_07_null")
    Parents="((Item=(Type=Bone,Name=\"deform_groin_l\"),Weight=0.0074),(Item=(Type=Bone,Name=\"deform_glute_r\"),Weight=0.0084),(Item=(Type=Bone,Name=\"spine_01\"),Weight=0.009),(Item=(Type=Bone,Name=\"spine_02\"),Weight=0.0109),(Item=(Type=Bone,Name=\"thigh_twist_01_l\"),Weight ...(truncated)"
    Weight=1
   )
   > &RigUnit_ParentConstraint(
    AdvancedSettings=(RotationOrderForFilter=XZY)
    Child=(Type=Null,Name="dyn_belt_08_null")
    Parents="((Item=(Type=Bone,Name=\"spine_03\"),Weight=0.0059),(Item=(Type=Bone,Name=\"spine_01\"),Weight=0.0092),(Item=(Type=Bone,Name=\"spine_02\"),Weight=0.0125),(Item=(Type=Bone,Name=\"dyn_shorts_out_01_r\"),Weight=0.0278),(Item=(Type=Bone,Name=\"deform_glute_r\"),Weight=0 ...(truncated)"
    Weight=1
   )
   > &RigUnit_ParentConstraint(
    AdvancedSettings=(RotationOrderForFilter=XZY)
    Child=(Type=Null,Name="dyn_belt_bk_03_null")
    Parents="((Item=(Type=Bone,Name=\"thigh_twist_01_r\"),Weight=0.015),(Item=(Type=Bone,Name=\"deform_glute_r\"),Weight=0.0261),(Item=(Type=Bone,Name=\"spine_01\"),Weight=0.0435),(Item=(Type=Bone,Name=\"thigh_twist_01_l\"),Weight=0.0553),(Item=(Type=Bone,Name=\"spine_03\"),Weig ...(truncated)"
    Weight=1
   )
   > &RigUnit_ParentConstraint(
    AdvancedSettings=(RotationOrderForFilter=XZY)
    Child=(Type=Null,Name="dyn_belt_bk_04_null")
    Parents="((Item=(Type=Bone,Name=\"spine_01\"),Weight=0.0271),(Item=(Type=Bone,Name=\"thigh_twist_01_l\"),Weight=0.0319),(Item=(Type=Bone,Name=\"spine_03\"),Weight=0.0335),(Item=(Type=Bone,Name=\"spine_02\"),Weight=0.0407),(Item=(Type=Bone,Name=\"thigh_twist_01_r\"),Weight=0. ...(truncated)"
    Weight=1
   )
   > &RigUnit_ParentConstraint(
    AdvancedSettings=(RotationOrderForFilter=XZY)
    Child=(Type=Null,Name="dyn_belt_bk_05_null")
    Parents="((Item=(Type=Bone,Name=\"spine_01\"),Weight=0.0099),(Item=(Type=Bone,Name=\"spine_03\"),Weight=0.0105),(Item=(Type=Bone,Name=\"spine_02\"),Weight=0.0141),(Item=(Type=Bone,Name=\"thigh_twist_01_l\"),Weight=0.017),(Item=(Type=Bone,Name=\"deform_glute_l\"),Weight=0.024 ...(truncated)"
    Weight=1
   )
   > &RigUnit_ParentConstraint(
    AdvancedSettings=(RotationOrderForFilter=XZY)
    Child=(Type=Null,Name="dyn_belt_bk_06_null")
    Parents=((Item=(Type=Bone,Name="deform_groin_r"),Weight=0.0076),(Item=(Type=Bone,Name="thigh_twist_01_r"),Weight=0.2542),(Item=(Type=Bone,Name="deform_glute_r"),Weight=0.2645),(Item=(Type=Bone,Name="pelvis"),Weight=0.4737))
    Weight=1
   )
   > &RigUnit_ParentConstraint(
    AdvancedSettings=(RotationOrderForFilter=XZY)
    Child=(Type=Null,Name="dyn_belt_bk_07_null")
    Parents="((Item=(Type=Bone,Name=\"spine_01\"),Weight=0.0059),(Item=(Type=Bone,Name=\"spine_02\"),Weight=0.0077),(Item=(Type=Bone,Name=\"dyn_shorts_out_01_r\"),Weight=0.029),(Item=(Type=Bone,Name=\"deform_groin_r\"),Weight=0.0453),(Item=(Type=Bone,Name=\"deform_glute_r\"),Wei ...(truncated)"
    Weight=1
   )
   > &RigUnit_ParentConstraint(
    AdvancedSettings=(RotationOrderForFilter=XZY)
    Child=(Type=Null,Name="dyn_belt_bk_08_null")
    Parents="((Item=(Type=Bone,Name=\"spine_02\"),Weight=0.0052),(Item=(Type=Bone,Name=\"deform_groin_r\"),Weight=0.0703),(Item=(Type=Bone,Name=\"dyn_shorts_out_01_r\"),Weight=0.0793),(Item=(Type=Bone,Name=\"deform_glute_r\"),Weight=0.0907),(Item=(Type=Bone,Name=\"pelvis\"),Weig ...(truncated)"
    Weight=1
   )
   > RigVMFunctionReturnNode()
  ]}
 New Function_ContainedGraph {nodes [
   RigVMFunctionEntryNode()
   > RigVMFunction_Sequence = RigVMAggregateNode()
   DISPATCH_RigVMDispatch_ArrayIterator = RigVMDispatchNode(ExecuteContext=@RigVMFunction_Sequence.A, Array=RigVMVariableNode() {variable=belt_ctrls}.Value) {template="DISPATCH_RigVMDispatch_ArrayIterator(in Array,out Element,out Index,out Count,out Ratio)"}
   n1 = RigVMVariableNode() {variable=Control Settings}
   > /Script/ControlRig.RigUnit_HierarchySetShapeSettings(Item=@DISPATCH_RigVMDispatch_ArrayIterator.Element, Settings=@n1.Value) {is=struct RigVMUnitNode}
   n2 = RigVMDispatchNode(ExecuteContext=@RigVMFunction_Sequence.B, Array=RigVMVariableNode() {variable=belt_bk_ctrls}.Value) {template="DISPATCH_RigVMDispatch_ArrayIterator(in Array,out Element,out Index,out Count,out Ratio)"}
   > &RigUnit_HierarchySetShapeSettings(Item=@n2.Element, Settings=@n1.Value)
   RigVMFunctionReturnNode(ExecuteContext=@RigVMFunction_Sequence.D)
  ]}
 New Function_ContainedGraph {nodes [
   n1 = RigVMFunctionEntryNode()
   RigVMFunction_Sequence = &RigVMFunction_Sequence(ExecuteContext=@n1.ExecuteContext)
   &RigUnit_ModifyTransforms(
    ItemToModify=((Item=(Type=Bone,Name="dyn_belt_anchor_r"),Transform=(Rotation=(X=0.287,Y=-0.0319,Z=0.0282,W=0.957),Translation=(X=-2.7857,Y=-4.6875,Z=3.7771),Scale3D=(X=1,Y=1,Z=1))))
    Weight=1
    WeightMinimum=0
    ExecutePin=@RigVMFunction_Sequence.A
   )
   > n2 = &RigUnit_SphericalPoseReader(
    ActiveRegionScaleFactors=(PositiveWidth=1,NegativeWidth=1,PositiveHeight=1,NegativeHeight=1)
    ActiveRegionSize=0.1
    Debug=(bDrawDebug=false,bDraw2D=False,bDrawLocalAxes=False,DebugScale=25,DebugSegments=20,DebugThickness=0.25)
    DriverAxis=(X=1,Y=0,Z=0)
    DriverItem=(Type=Bone,Name="thigh_r")
    FalloffRegionScaleFactors=(PositiveWidth=1,NegativeWidth=1,PositiveHeight=1,NegativeHeight=1)
    FalloffSize=0.6
    FlipHeightScaling=False
    FlipWidthScaling=False
    OptionalParentItem=(Type=Bone,Name="pelvis")
    RotationOffset=(X=5.6525,Y=-75.9752,Z=-91.3442)
   )
   RigVMFunction_AlphaInterp = &RigVMFunction_AlphaInterp(InterpSpeedDecreasing=20, InterpSpeedIncreasing=20, Scale=1, bClampResult=False, bInterpResult=True, Value=@n2.OutputParam)
   > &RigUnit_ModifyTransforms(
    ItemToModify="((Item=(Type=Bone,Name=\"dyn_belt_anchor_r\"),Transform=(Rotation=(X=-0.0659,Y=0.0343,Z=-0.0366,W=0.9966),Translation=(X=-1.6777,Y=-7.9922,Z=-4.3213),Scale3D=(X=1,Y=1,Z=1))),(Item=(Type=Bone,Name=\"dyn_belt_root\"),Transform=(Rotation=(X=0,Y=0,Z=0,W=1),Transla ...(truncated)"
    WeightMinimum=0
    Weight=@RigVMFunction_AlphaInterp.Result
   )
   > &RigUnit_ModifyTransforms(
    ItemToModify=((Item=(Type=Bone,Name="dyn_belt_anchor_r"),Transform=(Rotation=(X=0,Y=0,Z=0.0532,W=0.9986),Translation=(X=0,Y=2.8997,Z=-2.0291),Scale3D=(X=1,Y=1,Z=1))))
    WeightMinimum=0
    Weight=RigVMVariableNode() {variable=isInFrontEnd}.Value
   )
   > SplineIK = RigVMFunctionReferenceNode(
    Primary Axis=(X=1,Y=0,Z=0)
    Samples Per Segment=6
    Secondary Spline Direction=(X=0,Y=1,Z=0)
    Spline Mode=BSpline
    Stretch=false
    Up Axis=(X=0,Y=1,Z=0)
    debug=false
    Controls=RigVMVariableNode() {variable=belt_ctrls_RBAN}.Value
    Bones=RigVMVariableNode() {variable=belt_joints_RBAN}.Value
   )
   > SplineIK_1 = RigVMFunctionReferenceNode(
    Primary Axis=(X=1,Y=0,Z=0)
    Samples Per Segment=6
    Secondary Spline Direction=(X=0,Y=-1,Z=0)
    Spline Mode=BSpline
    Up Axis=(X=0,Y=-1,Z=0)
    debug=false
    Controls=RigVMVariableNode() {variable=belt_bk_ctrls_RBAN}.Value
    Bones=RigVMVariableNode() {variable=belt_bk_joints_RBAN}.Value
    Stretch=RigVMVariableNode() {variable=isSkydiving}.Value
   )
   > &RigUnit_ModifyTransforms(
    ItemToModify=((Item=(Type=Bone,Name="dyn_belt_anchor_r_parent"),Transform=(Rotation=(X=0,Y=0,Z=0,W=1),Translation=(X=0,Y=0,Z=61.4372),Scale3D=(X=1,Y=1,Z=1))))
    Weight=1
    WeightMinimum=0
   )
   RigVMFunctionReturnNode(ExecuteContext=@RigVMFunction_Sequence.B)
   &RigUnit_ParentConstraint(
    AdvancedSettings=(InterpolationType=Average,RotationOrderForFilter=XZY)
    Child=(Type=Bone,Name="dyn_belt_anchor_r")
    ExecutePin=()
    Parents=((Item=(Type=Bone,Name="pelvis"),Weight=1),(Item=(Type=Bone,Name="thigh_r"),Weight=1))
    Weight=1
   )
   /Script/ControlRigSpline.RigUnit_PositionFromControlRigSpline(U=0, Spline=@SplineIK_1.Output Spline) {is=struct RigVMUnitNode}
   /Script/ControlRigSpline.RigUnit_TangentFromControlRigSpline(U=0, Spline=@SplineIK_1.Output Spline) {is=struct RigVMUnitNode}
  ]}
 New Function_ContainedGraph {nodes [
   RigVMFunctionEntryNode()
   > n1 = &RigUnit_SphericalPoseReader(
    ActiveRegionScaleFactors=(PositiveWidth=1,NegativeWidth=1,PositiveHeight=1,NegativeHeight=1)
    ActiveRegionSize=0.1
    Debug=(bDrawDebug=false,bDraw2D=False,bDrawLocalAxes=False,DebugScale=25,DebugSegments=20,DebugThickness=0.25)
    DriverAxis=(X=1,Y=0,Z=0)
    DriverItem=(Type=Bone,Name="thigh_r")
    FalloffRegionScaleFactors=(PositiveWidth=1,NegativeWidth=1,PositiveHeight=1,NegativeHeight=1)
    FalloffSize=0.6
    FlipHeightScaling=False
    FlipWidthScaling=False
    OptionalParentItem=(Type=Bone,Name="pelvis")
    RotationOffset=(X=5.6525,Y=-75.9752,Z=-91.3442)
   )
   RigVMFunction_AlphaInterp = &RigVMFunction_AlphaInterp(InterpSpeedDecreasing=6, InterpSpeedIncreasing=6, Scale=1, bClampResult=False, bInterpResult=True, Value=@n1.OutputParam)
   > &RigUnit_ModifyTransforms(
    ItemToModify="((Item=(Type=Null,Name=\"dyn_belt_08_null\"),Transform=(Rotation=(X=0,Y=0,Z=0,W=1),Translation=(X=0,Y=-16.3311,Z=-3.044),Scale3D=(X=1,Y=1,Z=1))),(Item=(Type=Null,Name=\"dyn_belt_bk_08_null\"),Transform=(Rotation=(X=0,Y=0,Z=0,W=1),Translation=(X=0,Y=-5.1196,Z=8 ...(truncated)"
    WeightMinimum=0
    Weight=@RigVMFunction_AlphaInterp.Result
   )
   > SplineIK = RigVMFunctionReferenceNode(
    Primary Axis=(X=1,Y=0,Z=0)
    Samples Per Segment=3
    Secondary Spline Direction=(X=0,Y=1,Z=0)
    Spline Mode=BSpline
    Stretch=true
    Up Axis=(X=0,Y=1,Z=0)
    debug=false
    Controls=RigVMVariableNode() {variable=belt_ctrls}.Value
    Bones=RigVMVariableNode() {variable=belt_joints}.Value
   )
   > SplineIK_1 = RigVMFunctionReferenceNode(
    Primary Axis=(X=1,Y=0,Z=0)
    Samples Per Segment=3
    Secondary Spline Direction=(X=0,Y=-1,Z=0)
    Spline Mode=BSpline
    Stretch=true
    Up Axis=(X=0,Y=-1,Z=0)
    debug=false
    Controls=RigVMVariableNode() {variable=belt_bk_ctrls}.Value
    Bones=RigVMVariableNode() {variable=belt_bk_joints}.Value
   )
   > RigVMFunctionReturnNode()
  ]}
 RigVMFunctionLibrary {nodes [
   Build Belt = RigVMCollapseNode(ExecuteContext=())
   Pin Default = RigVMCollapseNode(ExecuteContext=())
   Pin No RBAN = RigVMCollapseNode(ExecuteContext=())
   Solve Belt No RBAN = RigVMCollapseNode(ExecuteContext=())
   Solve Belt RBAN = RigVMCollapseNode(ExecuteContext=())
  ]}
 RigVMFunction_Sequence_1_ContainedGraph {nodes [
   n1 = RigVMFunctionEntryNode()
   RigVMFunction_Sequence = &RigVMFunction_Sequence(ExecuteContext=@n1.ExecuteContext)
   n2 = &RigVMFunction_Sequence(ExecuteContext=@RigVMFunction_Sequence.B)
   n3 = &RigVMFunction_Sequence(ExecuteContext=@n2.B)
   RigVMFunctionReturnNode(A=@RigVMFunction_Sequence.A, B=@n2.A, C=@n3.A, D=@n3.B)
  ]}
 RigVMFunction_Sequence_1_ContainedGraph {nodes [
   n1 = RigVMFunctionEntryNode()
   RigVMFunction_Sequence = &RigVMFunction_Sequence(ExecuteContext=@n1.ExecuteContext)
   n2 = &RigVMFunction_Sequence(ExecuteContext=@RigVMFunction_Sequence.B)
   n3 = &RigVMFunction_Sequence(ExecuteContext=@n2.B)
   RigVMFunctionReturnNode(A=@RigVMFunction_Sequence.A, B=@n2.A, C=@n3.A, D=@n3.B)
  ]}
 RigVMModel {nodes [
   RigUnit_BeginExecution = &RigUnit_BeginExecution()
   RigVMFunction_Sequence = RigVMAggregateNode(ExecuteContext=@RigUnit_BeginExecution.ExecutePin)
   FN_DeformRig_FNC = RigVMFunctionReferenceNode(
    DeformRig_ Curve="(deformRig_l_upperarm_up_alpha_72_#1=False,deformRig_r_upperarm_up_alpha_73_#2=False,deformRig_l_upperarm_fr_alpha_74_#3=False,deformRig_r_upperarm_fr_alpha_75_CB51B8 ...(truncated)"
    DeformRig_Bool="(upperarm_1_#4=True,elbow_3_#5=True,wrist_7_#6=True,groin_8_#7=True,glute_17_#8=True,knee_9_40C408324CA261 ...(truncated)"
    DeformRig_Debug=(upperarm_debug_18_#4=False,wrist_debug_20_#6=False,lat_debug_26_#9=False,groinGlute_debug_27_#7=False)
    DeformRig_Value="(upperarm_up_18_#10=(Rotation=(X=0,Y=0,Z=0,W=1),Translation=(X=0,Y=0,Z=8),Scale3D=(X=0.7,Y=1,Z=1)),upperarm_fwd_fr_19_#11=(Rotation=(X=0,Y=0,Z=0,W=1),Translation=(X=0,Y=10,Z=0),Scale3D=(X=0.5,Y=1,Z= ...(truncated)"
    ExecuteContext=@RigVMFunction_Sequence.A
   )
   RigVMFunction_ControlFlowBranch = &RigVMFunction_ControlFlowBranch(ExecuteContext=@RigVMFunction_Sequence.B, Condition=RigVMVariableNode() {variable=RBAN Enabled}.Value)
   Pin Default_1 = RigVMFunctionReferenceNode(ExecuteContext=@RigVMFunction_ControlFlowBranch.False)
   > Pin No RBAN = RigVMFunctionReferenceNode()
   > Solve Belt No RBAN = RigVMFunctionReferenceNode()
   Pin Default = RigVMFunctionReferenceNode(ExecuteContext=@RigVMFunction_ControlFlowBranch.True)
   > Solve Belt RBAN = RigVMFunctionReferenceNode()
   n1 = RigVMVariableNode() {variable=shorts offset}
   &RigUnit_ModifyTransforms(
    ItemToModify="((Item=(Type=Bone,Name=\"dyn_shorts_bk_mstr_l\"),Transform=(Rotation=(X=0,Y=0,Z=0,W=1),Translation=(X=-6,Y=0,Z=0),Scale3D=(X=1,Y=1,Z=1))),(Item=(Type=Bone,Name=\"dyn_shorts_bk_mstr_r\"),Transform=(Rotation=(X=0,Y=0,Z=0,W=1),Translation=(X=-6,Y=0,Z=0),Scale3D=( ...(truncated)"
    Weight=1
    WeightMinimum=0
    ExecutePin=@RigVMFunction_Sequence.C
    ItemToModify.0.Transform.Translation=@n1.Value
    ItemToModify.1.Transform.Translation=@n1.Value
    ItemToModify.3.Transform.Translation=@n1.Value
    ItemToModify.4.Transform.Translation=@n1.Value
    ItemToModify.5.Transform.Translation=@n1.Value
    ItemToModify.6.Transform.Translation=@n1.Value
    ItemToModify.7.Transform.Translation=@n1.Value
   )
   > n2 = &RigUnit_SphericalPoseReader(
    ActiveRegionScaleFactors=(PositiveWidth=1,NegativeWidth=1,PositiveHeight=1,NegativeHeight=1)
    ActiveRegionSize=0.1
    Debug=(bDrawDebug=false,bDraw2D=False,bDrawLocalAxes=False,DebugScale=25,DebugSegments=20,DebugThickness=0.25)
    DriverAxis=(X=-1,Y=0,Z=0)
    DriverItem=(Type=Bone,Name="calf_l")
    FalloffRegionScaleFactors=(PositiveWidth=1,NegativeWidth=1,PositiveHeight=1,NegativeHeight=1)
    FalloffSize=0.5
    FlipHeightScaling=False
    FlipWidthScaling=False
    OptionalParentItem=(Type=Bone,Name="thigh_l")
    RotationOffset=(X=0,Y=90,Z=118.878)
   )
   > &RigUnit_ModifyTransforms(
    ItemToModify="((Item=(Type=Bone,Name=\"dyn_shorts_bk_mstr_l\"),Transform=(Rotation=(X=0,Y=0,Z=0.1654,W=0.9862),Translation=(X=-2,Y=0,Z=0),Scale3D=(X=1,Y=1,Z=1))),(Item=(Type=Bone,Name=\"dyn_shorts_out_mstr_l\"),Transform=(Rotation=(X=0,Y=0,Z=0,W=1),Translation=(X=-4,Y=0,Z=0 ...(truncated)"
    WeightMinimum=0
    Weight=@n2.OutputParam
   )
   > n3 = &RigUnit_SphericalPoseReader(
    ActiveRegionScaleFactors=(PositiveWidth=1,NegativeWidth=1,PositiveHeight=1,NegativeHeight=1)
    ActiveRegionSize=0.2
    Debug=(bDrawDebug=false,bDraw2D=False,bDrawLocalAxes=False,DebugScale=25,DebugSegments=20,DebugThickness=0.25)
    DriverAxis=(X=1,Y=0,Z=0)
    DriverItem=(Type=Bone,Name="calf_r")
    FalloffRegionScaleFactors=(PositiveWidth=1,NegativeWidth=1,PositiveHeight=1,NegativeHeight=1)
    FalloffSize=0.5
    FlipHeightScaling=False
    FlipWidthScaling=False
    OptionalParentItem=(Type=Bone,Name="thigh_r")
    RotationOffset=(X=7.9727,Y=-88.4358,Z=129.1277)
   )
   n4 = &RigVMFunction_AlphaInterp(InterpSpeedDecreasing=20, InterpSpeedIncreasing=20, Scale=1, bClampResult=False, bInterpResult=True, Value=@n3.OutputParam)
   > &RigUnit_ModifyTransforms(
    ItemToModify="((Item=(Type=Bone,Name=\"dyn_shorts_bk_mstr_r\"),Transform=(Rotation=(X=0,Y=0,Z=0.3477,W=0.9376),Translation=(X=-4,Y=-7.4928,Z=-0.0002),Scale3D=(X=1,Y=1,Z=1))),(Item=(Type=Bone,Name=\"dyn_shorts_out_mstr_r\"),Transform=(Rotation=(X=0.1317,Y=0.1393,Z=-0.0187,W= ...(truncated)"
    WeightMinimum=0
    Weight=@n4.Result
   )
   > n5 = &RigUnit_SphericalPoseReader(
    ActiveRegionScaleFactors=(PositiveWidth=1,NegativeWidth=1,PositiveHeight=1,NegativeHeight=1)
    ActiveRegionSize=0.01
    Debug=(bDrawDebug=false,bDraw2D=False,bDrawLocalAxes=False,DebugScale=25,DebugSegments=20,DebugThickness=0.25)
    DriverAxis=(X=0,Y=1,Z=0)
    DriverItem=(Type=Bone,Name="foot_l")
    FalloffRegionScaleFactors=(PositiveWidth=1,NegativeWidth=1,PositiveHeight=1,NegativeHeight=1)
    FalloffSize=0.4
    FlipHeightScaling=False
    FlipWidthScaling=False
    OptionalParentItem=(Type=Bone,Name="calf_l")
    RotationOffset=(X=90,Y=0,Z=-61.0573)
   )
   n6 = &RigVMFunction_AlphaInterp(InterpSpeedDecreasing=10, InterpSpeedIncreasing=10, Scale=1, bClampResult=False, bInterpResult=True, Value=@n5.OutputParam)
   > &RigUnit_ModifyTransforms(
    ItemToModify=((Item=(Type=Bone,Name="dyn_leg_armor_mstr_l"),Transform=(Rotation=(X=0,Y=0,Z=0,W=1),Translation=(X=0,Y=-5.8646,Z=0),Scale3D=(X=1,Y=1,Z=1))))
    WeightMinimum=0
    Weight=@n6.Result
   )
   > n7 = &RigUnit_SphericalPoseReader(
    ActiveRegionScaleFactors=(PositiveWidth=1,NegativeWidth=1,PositiveHeight=1,NegativeHeight=1)
    ActiveRegionSize=0.02
    Debug=(bDrawDebug=false,bDraw2D=False,bDrawLocalAxes=False,DebugScale=25,DebugSegments=20,DebugThickness=0.25)
    DriverAxis=(X=0,Y=-1,Z=0)
    DriverItem=(Type=Bone,Name="foot_r")
    FalloffRegionScaleFactors=(PositiveWidth=1,NegativeWidth=1,PositiveHeight=1,NegativeHeight=1)
    FalloffSize=0.4
    FlipHeightScaling=False
    FlipWidthScaling=False
    OptionalParentItem=(Type=Bone,Name="calf_l")
    RotationOffset=(X=82.593,Y=2.7292,Z=-216.6609)
   )
   n8 = &RigVMFunction_AlphaInterp(InterpSpeedDecreasing=10, InterpSpeedIncreasing=10, Scale=1, bClampResult=False, bInterpResult=True, Value=@n7.OutputParam)
   > &RigUnit_ModifyTransforms(
    ItemToModify="((Item=(Type=Bone,Name=\"dyn_leg_armor_mstr_r\"),Transform=(Rotation=(X=0,Y=0,Z=0,W=1),Translation=(X=0,Y=-9.0113,Z=0),Scale3D=(X=1,Y=1,Z=1))),(Item=(Type=Bone,Name=\"dyn_foot_armor_mstr_r\"),Transform=(Rotation=(X=0,Y=0,Z=0.1431,W=0.9897),Translation=(X=-0.46 ...(truncated)"
    WeightMinimum=0
    Weight=@n8.Result
   )
   > n9 = &RigUnit_SphericalPoseReader(
    ActiveRegionScaleFactors=(PositiveWidth=1,NegativeWidth=1,PositiveHeight=1,NegativeHeight=1)
    ActiveRegionSize=0.1
    Debug=(bDrawDebug=false,bDraw2D=False,bDrawLocalAxes=False,DebugScale=25,DebugSegments=20,DebugThickness=0.25)
    DriverAxis=(X=-1,Y=0,Z=0)
    DriverItem=(Type=Bone,Name="upperarm_r")
    FalloffRegionScaleFactors=(PositiveWidth=1,NegativeWidth=1,PositiveHeight=1,NegativeHeight=1)
    FalloffSize=0.5
    FlipHeightScaling=False
    FlipWidthScaling=False
    OptionalParentItem=(Type=Bone,Name="None")
    RotationOffset=(X=-117.6404,Y=4.3612,Z=33.075)
   )
   > &RigUnit_ModifyTransforms(
    ItemToModify="((Item=(Type=Bone,Name=\"dyn_shld_armor_mstr_r\"),Transform=(Rotation=(X=-0.3287,Y=-0.1747,Z=-0.0468,W=0.9269),Translation=(X=-5.3538,Y=-2.6411,Z=2.5808),Scale3D=(X=1,Y=1,Z=1))),(Item=(Type=Bone,Name=\"dyn_shld_plate_r\"),Transform=(Rotation=(X=-0.0911,Y=-0.08 ...(truncated)"
    WeightMinimum=0
    Weight=@n9.OutputParam
   )
   > &RigUnit_SphericalPoseReader(
    ActiveRegionScaleFactors=(PositiveWidth=1,NegativeWidth=1,PositiveHeight=1,NegativeHeight=1)
    ActiveRegionSize=0.05
    Debug=(bDrawDebug=false,bDraw2D=False,bDrawLocalAxes=False,DebugScale=25,DebugSegments=20,DebugThickness=0.25)
    DriverAxis=(X=-1,Y=0,Z=0)
    DriverItem=(Type=Bone,Name="upperarm_r")
    FalloffRegionScaleFactors=(PositiveWidth=1,NegativeWidth=1,PositiveHeight=1,NegativeHeight=1)
    FalloffSize=0.2
    FlipHeightScaling=False
    FlipWidthScaling=False
    OptionalParentItem=(Type=Bone,Name="None")
    RotationOffset=(X=-128.9614,Y=4.3612,Z=-75.428)
   )
   > &RigUnit_ModifyTransforms(
    ItemToModify="((Item=(Type=Bone,Name=\"dyn_shld_armor_mstr_r\"),Transform=(Rotation=(X=0,Y=0,Z=0,W=1),Translation=(X=3.6282,Y=0,Z=0),Scale3D=(X=1,Y=1,Z=1))),(Item=(Type=Bone,Name=\"dyn_shld_plate_r\"),Transform=(Rotation=(X=0,Y=0,Z=0,W=1),Translation=(X=3.3459,Y=0,Z=0),Scal ...(truncated)"
    Weight=1
    WeightMinimum=0
   )
   > n10 = &RigUnit_SphericalPoseReader(
    ActiveRegionScaleFactors=(PositiveWidth=0.5,NegativeWidth=0.75,PositiveHeight=0.5,NegativeHeight=1)
    ActiveRegionSize=0.2
    Debug=(bDrawDebug=false,bDraw2D=False,bDrawLocalAxes=False,DebugScale=25,DebugSegments=20,DebugThickness=0.25)
    DriverAxis=(X=1,Y=0,Z=0)
    DriverItem=(Type=Bone,Name="upperarm_l")
    FalloffRegionScaleFactors=(PositiveWidth=1,NegativeWidth=1,PositiveHeight=1,NegativeHeight=1)
    FalloffSize=0.6
    FlipHeightScaling=False
    FlipWidthScaling=False
    OptionalParentItem=(Type=Bone,Name="None")
    RotationOffset=(X=13.5237,Y=7.9318,Z=-68.3133)
   )
   n11 = &RigVMFunction_AlphaInterp(InterpSpeedDecreasing=40, InterpSpeedIncreasing=40, Scale=1, bClampResult=False, bInterpResult=True, Value=@n10.OutputParam)
   > &RigUnit_ModifyTransforms(
    ItemToModify="((Item=(Type=Bone,Name=\"dyn_shld_armor_mstr_l\"),Transform=(Rotation=(X=0.6965,Y=-0.0048,Z=0.1328,W=0.7052),Translation=(X=4.5127,Y=3.9112,Z=-15.3117),Scale3D=(X=1,Y=1,Z=1))),(Item=(Type=Bone,Name=\"dyn_shld_plate_01_mstr_l\"),Transform=(Rotation=(X=0.5468,Y= ...(truncated)"
    WeightMinimum=0
    Weight=@n11.Result
   )
   > n12 = &RigUnit_SphericalPoseReader(
    ActiveRegionScaleFactors=(PositiveWidth=1,NegativeWidth=1,PositiveHeight=1,NegativeHeight=1)
    ActiveRegionSize=0.05
    Debug=(bDrawDebug=false,bDraw2D=False,bDrawLocalAxes=False,DebugScale=25,DebugSegments=20,DebugThickness=0.25)
    DriverAxis=(X=1,Y=0,Z=0)
    DriverItem=(Type=Bone,Name="hand_l")
    FalloffRegionScaleFactors=(PositiveWidth=1,NegativeWidth=1,PositiveHeight=1,NegativeHeight=1)
    FalloffSize=0.6
    FlipHeightScaling=False
    FlipWidthScaling=False
    OptionalParentItem=(Type=Bone,Name="lowerarm_r")
    RotationOffset=(X=-133.1274,Y=0,Z=0)
   )
   n13 = &RigVMFunction_AlphaInterp(InterpSpeedDecreasing=20, InterpSpeedIncreasing=20, Scale=1, bClampResult=False, bInterpResult=True, Value=@n12.OutputParam)
   > &RigUnit_ModifyTransforms(
    ItemToModify="((Item=(Type=Bone,Name=\"dyn_gauntlet_plate_02_mstr\"),Transform=(Rotation=(X=0,Y=0,Z=0,W=1),Translation=(X=-4.3307,Y=0,Z=0),Scale3D=(X=1,Y=1,Z=1))),(Item=(Type=Bone,Name=\"dyn_gauntlet_plate_01_mstr\"),Transform=(Rotation=(X=0,Y=0,Z=0,W=1),Translation=(X=-3.6 ...(truncated)"
    WeightMinimum=0
    Weight=@n13.Result
   )
  ]}
 RigVMModel Construction Event Graph {nodes [
   n1 = &RigUnit_PrepareForExecution()
   Build Belt = RigVMFunctionReferenceNode(ExecuteContext=@n1.ExecutePin)
  ]}
]
hierarchy [
 root {loc=<0 0 0>, rot=<0 0 0 1>}
 pelvis {parent=root, loc=<0 -0.7071 93.1191>, rot=<0.0127 0.707 -0.0127 -0.707>}
 spine_01 {parent=pelvis, loc=<6.7494 -0.6388 0>, rot=<0 0 0.034 -0.9994>}
 spine_02 {parent=spine_01, loc=<7.2518 -0.035 0>, rot=<0 0 0.0238 -0.9997>}
 spine_03 {parent=spine_02, loc=<6.0104 0 0>, rot=<0 0 0.0149 -0.9999>}
 spine_04 {parent=spine_03, loc=<8.1024 0 0>, rot=<0 0 -0.0132 -0.9999>}
 spine_05 {parent=spine_04, loc=<10.0146 0 0>, rot=<0 0 -0.0015 -1>}
 attach_cape {parent=spine_05, loc=<9.008 -11.6861 -0.0142>, rot=<-0.5479 -0.4421 0.2654 -0.6588>}
 neck_01 {parent=spine_05, loc=<11.3138 0.1529 0>, rot=<0 0 -0.2003 -0.9797>}
 neck_02 {parent=neck_01, loc=<4.5627 0 0>, rot=<0 0 0.0504 -0.9987>}
 head {parent=neck_02, loc=<5.3914 0 0>, rot=<0 0 0.115 -0.9934>}
 attach_backpack {parent=spine_05, loc=<-0.1169 -13.4383 -0.011>, rot=<-0.0092 -0.7071 0.0092 -0.707>}
 clavicle_l {parent=spine_05, loc=<4.3575 -0.1466 -2.3223>, rot=<0.0465 -0.7289 -0.009 -0.6829>}
 upperarm_l {parent=clavicle_l, loc=<15.5111 0 0>, rot=<0.0226 -0.3234 -0.0088 -0.9459>}
 lowerarm_l {parent=upperarm_l, loc=<26.6514 0 0>, rot=<0.0391 -0.0099 -0.2453 -0.9686>}
 hand_l {parent=lowerarm_l, loc=<22.3491 0 0>, rot=<-0.7081 -0.0507 0.0975 -0.6975>}
 thumb_01_l {parent=hand_l, loc=<1.3118 -0.3191 -2.459>, rot=<0.6546 -0.4042 0.0664 -0.6354>}
 thumb_02_l {parent=thumb_01_l, loc=<5.9336 -0.0448 0.0401>, rot=<-0.0037 0.0213 0.298 -0.9543>}
 thumb_03_l {parent=thumb_02_l, loc=<2.2223 0.3682 -0.0907>, rot=<0.0118 -0.013 0.0066 -0.9998>}
 index_metacarpal_l {parent=hand_l, loc=<1.6535 0.0183 -2.96>, rot=<0.0158 -0.0124 -0.0173 -0.9996>}
 index_01_l {parent=index_metacarpal_l, loc=<8.3008 0.0357 -0.0524>, rot=<-0.0044 0.0485 0.0924 -0.9945>}
 index_02_l {parent=index_01_l, loc=<3.3587 -0.3004 0.0452>, rot=<-0.002 0.0027 0.3225 -0.9466>}
 index_03_l {parent=index_02_l, loc=<2.5578 -0.123 -0.0252>, rot=<-0.002 0.0067 0.2558 -0.9667>}
 middle_metacarpal_l {parent=hand_l, loc=<2.1856 -0.1725 -0.5984>, rot=<0.006 0.0168 -0.0329 -0.9993>}
 middle_01_l {parent=middle_metacarpal_l, loc=<7.3275 -0.2447 0.0129>, rot=<-0.0208 0.0042 0.0858 -0.9961>}
 middle_02_l {parent=middle_01_l, loc=<4.0123 0.1401 0.1175>, rot=<-0.0218 0.0258 0.3512 -0.9357>}
 middle_03_l {parent=middle_02_l, loc=<2.6077 0.0534 0.1373>, rot=<0.0082 0.0188 0.1708 -0.9851>}
 ring_metacarpal_l {parent=hand_l, loc=<2.2074 -0.3625 1.3638>, rot=<-0.0284 0.0722 -0.0458 -0.9959>}
 ring_01_l {parent=ring_metacarpal_l, loc=<7.0295 -0.6197 -0.1446>, rot=<-0.0489 -0.0252 0.1892 -0.9804>}
 ring_02_l {parent=ring_01_l, loc=<3.6174 0.4647 -0.1027>, rot=<0 0 0.2806 -0.9598>}
 ring_03_l {parent=ring_02_l, loc=<2.4152 0.114 0.0199>, rot=<0 0 0.1189 -0.9929>}
 pinky_metacarpal_l {parent=hand_l, loc=<1.7616 -0.9753 3.3814>, rot=<-0.0397 0.0814 -0.069 -0.9935>}
 pinky_01_l {parent=pinky_metacarpal_l, loc=<7.1709 -0.4831 -0.1094>, rot=<-0.0654 -0.0035 0.2242 -0.9723>}
 pinky_02_l {parent=pinky_01_l, loc=<2.8274 0.0893 -0.1043>, rot=<0.0037 0.0011 0.2854 -0.9584>}
 pinky_03_l {parent=pinky_02_l, loc=<2.2046 0.0307 -0.0078>, rot=<-0.0002 0.0098 0.1445 -0.9895>}
 deform_wrist_dn_base_l {parent=hand_l, loc=<0 -2.2342 0>, rot=<0 0 0 -1>}
 deform_wrist_dn_l {parent=deform_wrist_dn_base_l, loc=<0 0 0>, rot=<0 0 0 1>}
 deform_wrist_up_base_l {parent=hand_l, loc=<0 2.2682 0>, rot=<0 0 0 -1>}
 deform_wrist_up_l {parent=deform_wrist_up_base_l, loc=<0 0 0>, rot=<0 0 0 1>}
 lowerarm_twist_02_l {parent=lowerarm_l, loc=<9.6199 0 0>, rot=<0 0 0 -1>}
 lowerarm_twist_01_l {parent=lowerarm_l, loc=<17.8549 0 0>, rot=<0 0 0 -1>}
 dyn_gauntlet_plate_01_mstr {parent=lowerarm_twist_01_l, loc=<-5.0028 -0.6384 4.747>, rot=<0.6601 0.0016 0.003 -0.7512>}
 dyn_gauntlet_plate_01 {parent=dyn_gauntlet_plate_01_mstr, loc=<0 0 0>, rot=<0 0 0 1>}
 dyn_gauntlet_plate_02_mstr {parent=lowerarm_twist_01_l, loc=<1.9772 -0.6008 4.5992>, rot=<0.6652 -0.0233 -0.0084 -0.7463>}
 dyn_gauntlet_plate_02 {parent=dyn_gauntlet_plate_02_mstr, loc=<0 0 0>, rot=<0 0 0 1>}
 deform_elbow_in_base_l {parent=lowerarm_l, loc=<0.2112 5.2591 0.8274>, rot=<0 0 0 -1>}
 deform_elbow_in_l {parent=deform_elbow_in_base_l, loc=<0 0 0>, rot=<0 0 0 1>}
 deform_elbow_out_base_l {parent=lowerarm_l, loc=<0 -4 0>, rot=<0 0 0 -1>}
 deform_elbow_out_l {parent=deform_elbow_out_base_l, loc=<0 0 0>, rot=<0 0 0 1>}
 upperarm_twist_01_l {parent=upperarm_l, loc=<0.0041 0 0>, rot=<0 0 0 1>}
 dyn_shld_armor_mstr_l {parent=upperarm_twist_01_l, loc=<-1.7172 -0.4394 9.0397>, rot=<0.6902 0.0272 0.0277 -0.7226>}
 dyn_shld_armor_l {parent=dyn_shld_armor_mstr_l, loc=<0 0 0>, rot=<0 0 0 1>}
 dyn_shld_plate_01_mstr_l {parent=upperarm_twist_01_l, loc=<8.0927 -0.1706 3.797>, rot=<0.6899 -0.0091 -0.0086 -0.7238>}
 dyn_shld_plate_01_l {parent=dyn_shld_plate_01_mstr_l, loc=<0 0 0>, rot=<0 0 0 1>}
 deform_upperarm_bk_base_l {parent=upperarm_twist_01_l, loc=<1.0813 -7.2192 -1.9453>, rot=<0 0 0 -1>}
 deform_upperarm_bk_l {parent=deform_upperarm_bk_base_l, loc=<0 0 0>, rot=<0 0 0 1>}
 deform_upperarm_fr_base_l {parent=upperarm_twist_01_l, loc=<-0.3691 7.0507 -2.765>, rot=<0 0 0 -1>}
 deform_upperarm_fr_l {parent=deform_upperarm_fr_base_l, loc=<0 0 0>, rot=<0 0 0 1>}
 deform_upperarm_up_base_l {parent=upperarm_twist_01_l, loc=<-3.2216 -1.1391 3.3311>, rot=<0 0 0 -1>}
 deform_upperarm_up_l {parent=deform_upperarm_up_base_l, loc=<0 0 0>, rot=<0 0 0 1>}
 dyn_shld_plate_02_mstr_l {parent=upperarm_twist_01_l, loc=<12.4222 -0.164 3.617>, rot=<0.6899 -0.0091 -0.0086 -0.7238>}
 dyn_shld_plate_02_l {parent=dyn_shld_plate_02_mstr_l, loc=<0 0 0>, rot=<0 0 0 1>}
 upperarm_twist_02_l {parent=upperarm_l, loc=<13.7941 0 0>, rot=<0 0 0 1>}
 clavicle_r {parent=spine_05, loc=<4.3575 -0.1466 2.3223>, rot=<0.7289 0.0465 -0.6829 0.009>}
 upperarm_r {parent=clavicle_r, loc=<-15.5111 0 0>, rot=<0.0226 -0.3234 -0.0088 -0.9459>}
 lowerarm_r {parent=upperarm_r, loc=<-26.6514 0 0>, rot=<0.0391 -0.0099 -0.2453 -0.9686>}
 hand_r {parent=lowerarm_r, loc=<-22.3491 0 0>, rot=<-0.7081 -0.0507 0.0975 -0.6975>}
 thumb_01_r {parent=hand_r, loc=<-1.3118 0.3191 2.459>, rot=<0.6546 -0.4042 0.0664 -0.6354>}
 thumb_02_r {parent=thumb_01_r, loc=<-5.9336 0.0448 -0.0401>, rot=<-0.0037 0.0213 0.298 -0.9543>}
 thumb_03_r {parent=thumb_02_r, loc=<-2.2223 -0.3682 0.0907>, rot=<0.0118 -0.013 0.0066 -0.9998>}
 index_metacarpal_r {parent=hand_r, loc=<-1.6535 -0.0183 2.96>, rot=<0.0158 -0.0124 -0.0173 -0.9996>}
 index_01_r {parent=index_metacarpal_r, loc=<-8.3008 -0.0357 0.0524>, rot=<-0.0044 0.0485 0.0924 -0.9945>}
 index_02_r {parent=index_01_r, loc=<-3.3587 0.3004 -0.0452>, rot=<-0.002 0.0027 0.3225 -0.9466>}
 index_03_r {parent=index_02_r, loc=<-2.5578 0.123 0.0252>, rot=<-0.002 0.0067 0.2558 -0.9667>}
 middle_metacarpal_r {parent=hand_r, loc=<-2.1856 0.1725 0.5984>, rot=<0.006 0.0168 -0.0329 -0.9993>}
 middle_01_r {parent=middle_metacarpal_r, loc=<-7.3275 0.2447 -0.0129>, rot=<-0.0208 0.0042 0.0858 -0.9961>}
 middle_02_r {parent=middle_01_r, loc=<-4.0123 -0.1401 -0.1175>, rot=<-0.0218 0.0258 0.3512 -0.9357>}
 middle_03_r {parent=middle_02_r, loc=<-2.6077 -0.0534 -0.1373>, rot=<0.0082 0.0188 0.1708 -0.9851>}
 ring_metacarpal_r {parent=hand_r, loc=<-2.2074 0.3625 -1.3638>, rot=<-0.0284 0.0722 -0.0458 -0.9959>}
 ring_01_r {parent=ring_metacarpal_r, loc=<-7.0295 0.6197 0.1446>, rot=<-0.0489 -0.0252 0.1892 -0.9804>}
 ring_02_r {parent=ring_01_r, loc=<-3.6174 -0.4647 0.1027>, rot=<0 0 0.2806 -0.9598>}
 ring_03_r {parent=ring_02_r, loc=<-2.4152 -0.114 -0.0199>, rot=<0 0 0.1189 -0.9929>}
 pinky_metacarpal_r {parent=hand_r, loc=<-1.7616 0.9753 -3.3814>, rot=<-0.0397 0.0814 -0.069 -0.9935>}
 pinky_01_r {parent=pinky_metacarpal_r, loc=<-7.1709 0.4831 0.1094>, rot=<-0.0654 -0.0035 0.2242 -0.9723>}
 pinky_02_r {parent=pinky_01_r, loc=<-2.8274 -0.0893 0.1043>, rot=<0.0037 0.0011 0.2854 -0.9584>}
 pinky_03_r {parent=pinky_02_r, loc=<-2.2046 -0.0307 0.0078>, rot=<-0.0002 0.0098 0.1445 -0.9895>}
 deform_wrist_dn_base_r {parent=hand_r, loc=<-0.0004 2.2345 -0.0001>, rot=<0 0 0 -1>}
 deform_wrist_dn_r {parent=deform_wrist_dn_base_r, loc=<0 0 0>, rot=<0 0 0 1>}
 deform_wrist_up_base_r {parent=hand_r, loc=<0.0003 -2.2684 0.0001>, rot=<0 0 0 -1>}
 deform_wrist_up_r {parent=deform_wrist_up_base_r, loc=<0 0 0>, rot=<0 0 0 1>}
 lowerarm_twist_02_r {parent=lowerarm_r, loc=<-9.6199 0 0>, rot=<0 0 0 -1>}
 lowerarm_twist_01_r {parent=lowerarm_r, loc=<-17.8549 0 0>, rot=<0 0 0 -1>}
 deform_elbow_in_base_r {parent=lowerarm_r, loc=<-0.2111 -5.2591 -0.8276>, rot=<0 0 0 -1>}
 deform_elbow_in_r {parent=deform_elbow_in_base_r, loc=<-0.0001 0 0.0002>, rot=<0 0 0 1>}
 deform_elbow_out_base_r {parent=lowerarm_r, loc=<0 4 -0.0001>, rot=<0 0 0 -1>}
 deform_elbow_out_r {parent=deform_elbow_out_base_r, loc=<0 0 0>, rot=<0 0 0 1>}
 upperarm_twist_01_r {parent=upperarm_r, loc=<-0.0041 0 0>, rot=<0 0 0 -1>}
 deform_upperarm_bk_base_r {parent=upperarm_twist_01_r, loc=<-1.0814 7.2192 1.9455>, rot=<0 0 0 -1>}
 deform_upperarm_bk_r {parent=deform_upperarm_bk_base_r, loc=<0 0 0>, rot=<0 0 0 1>}
 deform_upperarm_fr_base_r {parent=upperarm_twist_01_r, loc=<0.3689 -7.0507 2.7652>, rot=<0 0 0 -1>}
 deform_upperarm_fr_r {parent=deform_upperarm_fr_base_r, loc=<0 0 0>, rot=<0 0 0 1>}
 deform_upperarm_up_base_r {parent=upperarm_twist_01_r, loc=<3.2214 1.1391 -3.3308>, rot=<0 0 0 -1>}
 deform_upperarm_up_r {parent=deform_upperarm_up_base_r, loc=<0 0 0>, rot=<0 0 0 -1>}
 dyn_shld_armor_mstr_r {parent=upperarm_twist_01_r, loc=<-2.9069 0.3019 -6.5944>, rot=<-0.0092 0.6896 0.7241 0.0115>}
 dyn_shld_armor_r {parent=dyn_shld_armor_mstr_r, loc=<0 0 0>, rot=<0 0 0 1>}
 dyn_shld_plate_r {parent=upperarm_twist_01_r, loc=<-7.1347 1.2507 -3.841>, rot=<-0.0092 0.6896 0.7241 0.0115>}
 upperarm_twist_02_r {parent=upperarm_r, loc=<-13.7941 0 0>, rot=<0 0 0 -1>}
 deform_pec_aim_base_l {parent=spine_04, loc=<7.8732 12.6388 -2.4123>, rot=<0.0102 -0.4602 0.3979 -0.7936>}
 deform_pec_aim_l {parent=deform_pec_aim_base_l, loc=<0 0 0>, rot=<0 0 0 1>}
 deform_pec_base_l {parent=deform_pec_aim_l, loc=<10.766 0 0>, rot=<0 0 0 1>}
 deform_pec_l {parent=deform_pec_base_l, loc=<0.0549 0 0>, rot=<0 0 0 1>}
 deform_pec_aim_base_r {parent=spine_04, loc=<7.8727 12.6388 2.4123>, rot=<-0.4602 -0.0102 0.7936 0.3979>}
 deform_pec_aim_r {parent=deform_pec_aim_base_r, loc=<0 0 0>, rot=<0 0 0 1>}
 deform_pec_base_r {parent=deform_pec_aim_r, loc=<-10.7662 -0.0006 -0.0007>, rot=<0 0 0 1>}
 deform_pec_r {parent=deform_pec_base_r, loc=<0 0 0>, rot=<0 0 0 1>}
 thigh_l {parent=pelvis, loc=<-2.5693 -0.0109 -7.7501>, rot=<0.0012 0.0572 0.0155 -0.9982>}
 calf_l {parent=thigh_l, loc=<-39.0484 0 0>, rot=<0.0006 -0.0161 -0.0391 -0.9991>}
 foot_l {parent=calf_l, loc=<-42.561 0.1077 0.4292>, rot=<0.0418 -0.0455 0.0343 -0.9975>}
 ball_l {parent=foot_l, loc=<-6.1599 13.8264 0.6806>, rot=<0.0001 -0.0001 -0.7063 -0.7079>}
 dyn_foot_armor_mstr_l {parent=foot_l, loc=<-3.4027 9.2602 0.4621>, rot=<-0.0234 0.0085 -0.2394 -0.9706>}
 dyn_foot_armor_l {parent=dyn_foot_armor_mstr_l, loc=<0 0 0>, rot=<0 0 0 1>}
 calf_twist_02_l {parent=calf_l, loc=<-18.8064 -0.0967 0.0397>, rot=<0.0237 0.0095 -0.0035 -0.9997>}
 calf_twist_01_l {parent=calf_l, loc=<-35.813 -0.0967 0.0397>, rot=<0.0237 0.0095 -0.0035 -0.9997>}
 dyn_leg_armor_mstr_l {parent=calf_twist_01_l, loc=<1.243 4.2752 -0.2161>, rot=<-0.01 -0.0156 -0.6105 -0.7918>}
 dyn_leg_armor_l {parent=dyn_leg_armor_mstr_l, loc=<0 0 0>, rot=<0 0 0 1>}
 deform_knee_in_base_l {parent=calf_l, loc=<1.298 -5.2957 -0.0496>, rot=<0 0 0 -1>}
 deform_knee_in_l {parent=deform_knee_in_base_l, loc=<0 0 0>, rot=<0 0 0 1>}
 deform_knee_out_base_l {parent=calf_l, loc=<0 4 0>, rot=<0 0 0 -1>}
 deform_knee_out_l {parent=deform_knee_out_base_l, loc=<0 0 0>, rot=<0 0 0 1>}
 thigh_twist_01_l {parent=thigh_l, loc=<-21.6303 0 0>, rot=<0.0013 0 0 -1>}
 dyn_shorts_fr_mstr_l {parent=thigh_twist_01_l, loc=<3.3224 9.1845 -0.0091>, rot=<0.0129 -0.0014 0.9999 0.0024>}
 dyn_shorts_fr_01_l {parent=dyn_shorts_fr_mstr_l, loc=<0 0 0>, rot=<0 0 0 1>}
 dyn_shorts_fr_02_l {parent=dyn_shorts_fr_01_l, loc=<10.6752 0 0>, rot=<0 0 0 1>}
 dyn_shorts_fr_flap_01_l {parent=dyn_shorts_fr_02_l, loc=<12.5923 0.8513 -4.3171>, rot=<-0.2809 -0.0489 0.0315 -0.958>}
 dyn_shorts_out_mstr_l {parent=thigh_twist_01_l, loc=<2.2089 -1.4894 -9.2745>, rot=<0.0109 0.7061 0.708 -0.0074>}
 dyn_shorts_out_01_l {parent=dyn_shorts_out_mstr_l, loc=<0 0 0>, rot=<0 0 0 1>}
 dyn_shorts_out_02_l {parent=dyn_shorts_out_01_l, loc=<9.853 0 0>, rot=<0 0 0 1>}
 dyn_shorts_out_flap_01_l {parent=dyn_shorts_out_02_l, loc=<12.8877 0.7825 -5.0305>, rot=<-0.2305 -0.0674 0.0765 -0.9677>}
 dyn_shorts_out_flap_02_l {parent=dyn_shorts_out_02_l, loc=<13.4579 -0.3421 1.4289>, rot=<0.1019 0.0251 0.1516 -0.9829>}
 dyn_shorts_bk_mstr_l {parent=thigh_twist_01_l, loc=<3.2192 -12.0276 -0.066>, rot=<-0.0024 -0.9999 -0.0014 0.0129>}
 dyn_shorts_bk_01_l {parent=dyn_shorts_bk_mstr_l, loc=<0 0 0>, rot=<0 0 0 1>}
 dyn_shorts_bk_02_l {parent=dyn_shorts_bk_01_l, loc=<10.6752 0 0>, rot=<0 0 0 1>}
 dyn_shorts_bk_flap_01_l {parent=dyn_shorts_bk_02_l, loc=<12.3 1.406 -3.3851>, rot=<-0.2444 0.0611 0.0768 -0.9647>}
 dyn_shorts_bk_flap_02_l {parent=dyn_shorts_bk_02_l, loc=<12.9785 -0.4369 4.1887>, rot=<0.1899 0.0623 0.0366 -0.9791>}
 dyn_shorts_in_mstr_l {parent=thigh_twist_01_l, loc=<3.498 -1.5442 8.8161>, rot=<0.0074 -0.708 0.7061 0.0109>}
 dyn_shorts_in_01_l {parent=dyn_shorts_in_mstr_l, loc=<0 0 0>, rot=<0 0 0 1>}
 dyn_shorts_in_02_l {parent=dyn_shorts_in_01_l, loc=<10.6738 0 0>, rot=<0 0 0 1>}
 dyn_shorts_in_flap_01_l {parent=dyn_shorts_in_02_l, loc=<10.5122 1.9986 -4.8737>, rot=<-0.2465 0.0576 0.0602 -0.9656>}
 dyn_shorts_in_flap_02_l {parent=dyn_shorts_in_02_l, loc=<10.7719 0.8897 3.8045>, rot=<0.1783 0.0413 0.1278 -0.9748>}
 deform_groin_base_l {parent=thigh_l, loc=<-2.1908 8.1142 -1.7442>, rot=<0 0 0 -1>}
 deform_groin_l {parent=deform_groin_base_l, loc=<0 0 0>, rot=<0 0 0 1>}
 thigh_r {parent=pelvis, loc=<-2.5693 -0.0109 7.7501>, rot=<0.0572 -0.0012 0.9982 0.0155>}
 calf_r {parent=thigh_r, loc=<39.0484 0 0>, rot=<0.0006 -0.0161 -0.0391 -0.9991>}
 foot_r {parent=calf_r, loc=<42.561 -0.1077 -0.4292>, rot=<0.0418 -0.0455 0.0343 -0.9975>}
 ball_r {parent=foot_r, loc=<6.1599 -13.8264 -0.6806>, rot=<0.0001 -0.0001 -0.7063 -0.7079>}
 dyn_foot_armor_mstr_r {parent=foot_r, loc=<3.4027 -9.2608 -0.4555>, rot=<-0.0092 -0.0262 0.9705 -0.2394>}
 dyn_foot_armor_r {parent=dyn_foot_armor_mstr_r, loc=<0 0 0>, rot=<0 0 0 1>}
 calf_twist_02_r {parent=calf_r, loc=<18.8064 0.0967 -0.0397>, rot=<0.0237 0.0095 -0.0035 -0.9997>}
 calf_twist_01_r {parent=calf_r, loc=<35.813 0.0967 -0.0397>, rot=<0.0237 0.0095 -0.0035 -0.9997>}
 dyn_leg_armor_mstr_r {parent=calf_twist_01_r, loc=<-1.2419 -4.2759 0.2276>, rot=<0.0143 -0.0115 0.7919 -0.6104>}
 dyn_leg_armor_r {parent=dyn_leg_armor_mstr_r, loc=<0 0 0>, rot=<0 0 0 1>}
 deform_knee_in_base_r {parent=calf_r, loc=<-1.298 5.2957 0.0496>, rot=<0 0 0 -1>}
 deform_knee_in_r {parent=deform_knee_in_base_r, loc=<0 0 0>, rot=<0 0 0 1>}
 deform_knee_out_base_r {parent=calf_r, loc=<0 -4 0>, rot=<0 0 0 -1>}
 deform_knee_out_r {parent=deform_knee_out_base_r, loc=<0 0 0>, rot=<0 0 0 1>}
 thigh_twist_01_r {parent=thigh_r, loc=<21.6303 0 0>, rot=<0.0013 0 0 -1>}
 dyn_shorts_fr_mstr_r {parent=thigh_twist_01_r, loc=<-3.3224 -9.1845 0.0091>, rot=<-0.0014 -0.0129 0.0024 -0.9999>}
 dyn_shorts_fr_01_r {parent=dyn_shorts_fr_mstr_r, loc=<0 0 0>, rot=<0 0 0 1>}
 dyn_shorts_fr_02_r {parent=dyn_shorts_fr_01_r, loc=<10.5977 0 0>, rot=<0 0 0 1>}
 dyn_shorts_fr_flap_01_r {parent=dyn_shorts_fr_02_r, loc=<12.7034 0.5262 -0.9891>, rot=<-0.1812 -0.1035 0.0735 -0.9752>}
 dyn_shorts_out_mstr_r {parent=thigh_twist_01_r, loc=<-2.2089 1.4894 9.2745>, rot=<0.7061 -0.0109 -0.0074 -0.708>}
 dyn_shorts_out_01_r {parent=dyn_shorts_out_mstr_r, loc=<0 0 0>, rot=<0 0 0 1>}
 dyn_shorts_out_02_r {parent=dyn_shorts_out_01_r, loc=<9.7756 0 0>, rot=<0 0 0 1>}
 dyn_shorts_out_flap_01_r {parent=dyn_shorts_out_02_r, loc=<12.5922 -0.1746 -0.9488>, rot=<-0.0615 0.0451 0.1171 -0.9902>}
 dyn_shorts_bk_mstr_r {parent=thigh_twist_01_r, loc=<-3.2192 12.0276 0.066>, rot=<0.9999 -0.0024 -0.0129 -0.0014>}
 dyn_shorts_bk_01_r {parent=dyn_shorts_bk_mstr_r, loc=<0 0 0>, rot=<0 0 0 1>}
 dyn_shorts_bk_02_r {parent=dyn_shorts_bk_01_r, loc=<10.3214 0 0>, rot=<0 0 0 1>}
 dyn_shorts_bk_flap_01_r {parent=dyn_shorts_bk_02_r, loc=<12.4214 0.4825 -4.6671>, rot=<-0.2443 -0.0495 -0.0242 -0.9681>}
 dyn_shorts_bk_flap_02_r {parent=dyn_shorts_bk_02_r, loc=<12.4911 1.4825 4.6303>, rot=<0.2571 -0.0477 0.0391 -0.9644>}
 dyn_shorts_in_mstr_r {parent=thigh_twist_01_r, loc=<-3.498 1.5442 -8.8161>, rot=<-0.708 -0.0074 0.0109 -0.7061>}
 dyn_shorts_in_01_r {parent=dyn_shorts_in_mstr_r, loc=<0 0 0>, rot=<0 0 0 1>}
 dyn_shorts_in_02_r {parent=dyn_shorts_in_01_r, loc=<10.5964 0 0>, rot=<0 0 0 1>}
 dyn_shorts_in_flap_02_r {parent=dyn_shorts_in_02_r, loc=<14.3791 -0.5231 2.5811>, rot=<0.298 0.1357 0.0015 -0.9449>}
 dyn_shorts_in_flap_01_r {parent=dyn_shorts_in_02_r, loc=<13.5186 -1.3092 -2.6491>, rot=<0 0 0.16 -0.9871>}
 deform_groin_base_r {parent=thigh_r, loc=<2.1908 -8.1142 1.7442>, rot=<0 0 0 -1>}
 deform_groin_r {parent=deform_groin_base_r, loc=<0 0 0>, rot=<0 0 0 1>}
 deform_glute_base_l {parent=pelvis, loc=<-8.8064 -8.8222 -8.4287>, rot=<0.0275 0.0501 -0.4472 -0.8926>}
 deform_glute_l {parent=deform_glute_base_l, loc=<0 0 0>, rot=<0 0 0 1>}
 deform_glute_base_r {parent=pelvis, loc=<-8.8062 -8.8222 8.4287>, rot=<0.0501 -0.0275 0.8926 -0.4472>}
 deform_glute_r {parent=deform_glute_base_r, loc=<0 0 0>, rot=<0 0 0 1>}
 dyn_belt_root {parent=pelvis, loc=<6.2828 -1.703 -16.6554>, rot=<0.0017 0.1015 -0.7067 -0.7002>}
 dyn_belt_sim_l {parent=dyn_belt_root, loc=<-0.0104 2.4907 0.0121>, rot=<0.0034 -0.0007 -0.0039 -1>}
 dyn_belt_01 {parent=dyn_belt_sim_l, loc=<0 0 0>, rot=<0 0.0001 0 -1>}
 dyn_belt_02 {parent=dyn_belt_01, loc=<5.998 0.0001 0>, rot=<0.0272 0.2475 -0.0332 -0.9679>}
 dyn_belt_03 {parent=dyn_belt_02, loc=<6.5916 0.0004 -0.0005>, rot=<0.0398 0.272 -0.0177 -0.9613>}
 dyn_belt_04 {parent=dyn_belt_03, loc=<6.8984 -0.0069 0.0217>, rot=<0.0088 0.1335 -0.0786 -0.9879>}
 dyn_belt_05 {parent=dyn_belt_04, loc=<7.4016 -0.0439 -0.0235>, rot=<0.0094 0.1126 0.0482 -0.9924>}
 dyn_belt_06 {parent=dyn_belt_05, loc=<7.4169 -0.0515 -0.094>, rot=<0.0157 0.1226 0.0388 -0.9916>}
 dyn_belt_07 {parent=dyn_belt_06, loc=<6.5904 0.0804 -0.2893>, rot=<0.0278 0.225 -0.0007 -0.974>}
 dyn_belt_08 {parent=dyn_belt_07, loc=<5.3503 -0.4258 -0.9978>, rot=<0.0368 0.2565 -0.0469 -0.9647>}
 dyn_belt_bk_01 {parent=dyn_belt_sim_l, loc=<-5.1967 0.4302 2.3228>, rot=<-0.0552 -0.901 0.0155 0.43>}
 dyn_belt_bk_02 {parent=dyn_belt_bk_01, loc=<6.2511 0.0007 -0.0018>, rot=<-0.0228 -0.2242 -0.0367 -0.9736>}
 dyn_belt_bk_03 {parent=dyn_belt_bk_02, loc=<6.8085 0.101 -0.0668>, rot=<0.0544 -0.1351 -0.0829 -0.9859>}
 dyn_belt_bk_04 {parent=dyn_belt_bk_03, loc=<7.2994 -0.1188 0.0211>, rot=<-0.0083 -0.0566 0.02 -0.9982>}
 dyn_belt_bk_05 {parent=dyn_belt_bk_04, loc=<7.4703 -0.1918 0.0167>, rot=<-0.0827 -0.2294 0.0306 -0.9693>}
 dyn_belt_bk_06 {parent=dyn_belt_bk_05, loc=<7.3051 -0.0254 0.1709>, rot=<-0.049 -0.2254 0.0502 -0.9717>}
 dyn_belt_bk_07 {parent=dyn_belt_bk_06, loc=<5.7286 0.5246 0.4624>, rot=<-0.0298 -0.187 -0.0364 -0.9812>}
 dyn_belt_bk_08 {parent=dyn_belt_bk_07, loc=<4.7722 0.1665 0.6067>, rot=<-0.0163 -0.2439 -0.016 -0.9695>}
 dyn_belt_anchor_r {parent=pelvis, loc=<-4.3902 4.2734 17.6395>, rot=<0.1602 -0.0222 -0.6916 -0.704>}
 dyn_belt_sim_r {parent=dyn_belt_anchor_r, loc=<-0.0727 3.3621 2.2067>, rot=<-0.0107 -0.0002 0.01 -0.9999>}
 dyn_belt_anchor_r_null {parent=root, type=Null, loc=<-18.7661 0.8352 88.2108>, rot=<-0.609 -0.4746 0.3881 0.5033>}
 dyn_belt_anchor_r_parent {parent=root, loc=<0 0 -2.1501>, rot=<0 0 0 1>}
 dyn_belt_anchor_r_ctrl {controlType=EULER_TRANSFORM, parent=dyn_belt_anchor_r_parent, type=Control, loc=<0 0 0>, rot=<0 0 0 1>}
 dyn_belt_01_null {type=Null, loc=<16.7774 -1.7629 99.112>, rot=<0.5151 0.5353 -0.5358 -0.4013>}
 dyn_belt_01_ctrl {controlType=EULER_TRANSFORM, parent=dyn_belt_01_null, type=Control, loc=<16.7774 -1.7629 99.112>, rot=<-0.5151 -0.5353 0.5358 0.4013>}
 dyn_belt_02_null {type=Null, loc=<15.894 4.125 98.3785>, rot=<-0.3946 -0.615 0.6449 0.2242>}
 dyn_belt_02_ctrl {controlType=EULER_TRANSFORM, parent=dyn_belt_02_null, type=Control, loc=<55.9437 -3.19 81.5682>, rot=<-0.2367 -0.5442 0.6977 0.4013>}
 dyn_belt_03_null {type=Null, loc=<12.0181 9.2296 96.8411>, rot=<0.2237 0.6708 -0.7068 -0.0211>}
 dyn_belt_03_ctrl {controlType=EULER_TRANSFORM, parent=dyn_belt_03_null, type=Control, loc=<55.6551 0.7956 79.9669>, rot=<-0.0338 -0.6632 0.7133 0.2242>}
 dyn_belt_04_null {type=Null, loc=<5.7425 11.5322 94.8311>, rot=<-0.1328 -0.6945 0.7029 -0.0769>}
 dyn_belt_04_ctrl {controlType=EULER_TRANSFORM, parent=dyn_belt_04_null, type=Control, loc=<23.3164 4.0233 92.5614>, rot=<-0.0454 -0.7254 0.6865 0.0211>}
 dyn_belt_05_null {type=Null, loc=<-1.1396 12.0836 92.7121>, rot=<0.0493 0.6812 -0.7144 0.152>}
 dyn_belt_05_ctrl {controlType=EULER_TRANSFORM, parent=dyn_belt_05_null, type=Control, loc=<12.6437 12.2788 92.2>, rot=<-0.0351 0.6936 -0.7154 0.0769>}
 dyn_belt_06_null {type=Null, loc=<-7.9686 11.0037 90.7144>, rot=<0.0489 -0.6557 0.7165 -0.2329>}
 dyn_belt_06_ctrl {controlType=EULER_TRANSFORM, parent=dyn_belt_06_null, type=Control, loc=<8.402 11.0821 91.5792>, rot=<-0.1465 0.6794 -0.7028 0.152>}
 dyn_belt_07_null {type=Null, loc=<-14.096 8.2549 89.0878>, rot=<-0.2148 0.5924 -0.6849 0.3658>}
 dyn_belt_07_ctrl {controlType=EULER_TRANSFORM, parent=dyn_belt_07_null, type=Control, loc=<19.5984 6.8713 89.525>, rot=<-0.3646 0.662 -0.6121 0.2329>}
 dyn_belt_08_null {type=Null, loc=<-18.0661 3.5674 88.2249>, rot=<0.3843 -0.4983 0.613 -0.4777>}
 dyn_belt_08_ctrl {controlType=EULER_TRANSFORM, parent=dyn_belt_08_null, type=Control, loc=<18.1565 0.4955 89.7521>, rot=<-0.5171 0.5986 -0.4903 0.3658>}
 dyn_belt_bk_01_null {type=Null, loc=<15.3066 -7.2997 99.0679>, rot=<0.2309 -0.6145 0.6706 -0.3456>}
 dyn_belt_bk_01_ctrl {controlType=EULER_TRANSFORM, parent=dyn_belt_bk_01_null, type=Control, loc=<0 0 0>, rot=<0 0 0 1>}
 dyn_belt_bk_02_null {type=Null, loc=<11.214 -11.9724 98.3481>, rot=<-0.0441 0.6689 -0.706 0.2287>}
 dyn_belt_bk_02_ctrl {controlType=EULER_TRANSFORM, parent=dyn_belt_bk_02_null, type=Control, loc=<-39.2772 -6.2397 90.8773>, rot=<0.1346 0.5855 -0.7209 0.3456>}
 dyn_belt_bk_03_null {type=Null, loc=<5.0337 -14.6195 96.6589>, rot=<-0.0473 -0.6935 0.7025 -0.1526>}
 dyn_belt_bk_03_ctrl {controlType=EULER_TRANSFORM, parent=dyn_belt_bk_03_null, type=Control, loc=<-25.3091 -14.7867 93.2796>, rot=<0.1387 0.6711 -0.6915 0.2287>}
 dyn_belt_bk_04_null {type=Null, loc=<-1.7681 -15.6859 94.6645>, rot=<0.1025 0.6996 -0.6994 0.1038>}
 dyn_belt_bk_04_ctrl {controlType=EULER_TRANSFORM, parent=dyn_belt_bk_04_null, type=Control, loc=<-22.8842 -14.7407 92.4462>, rot=<-0.1561 -0.6866 0.6934 -0.1526>}
 dyn_belt_bk_05_null {type=Null, loc=<-8.7146 -15.699 92.5711>, rot=<-0.2737 -0.652 0.7041 0.0646>}
 dyn_belt_bk_05_ctrl {controlType=EULER_TRANSFORM, parent=dyn_belt_bk_05_null, type=Control, loc=<-56.5788 -2.725 76.5657>, rot=<-0.4378 -0.5952 0.6658 -0.1038>}
 dyn_belt_bk_06_null {type=Null, loc=<-14.6682 -12.5308 90.4409>, rot=<0.4069 0.607 -0.6393 -0.2394>}
 dyn_belt_bk_06_ctrl {controlType=EULER_TRANSFORM, parent=dyn_belt_bk_06_null, type=Control, loc=<-59.2421 -10.0794 72.0973>, rot=<-0.5316 -0.6382 0.5531 0.0646>}
 dyn_belt_bk_07_null {type=Null, loc=<-18.2376 -7.379 88.9619>, rot=<-0.4988 -0.5505 0.5525 0.3781>}
 dyn_belt_bk_07_ctrl {controlType=EULER_TRANSFORM, parent=dyn_belt_bk_07_null, type=Control, loc=<-55.9808 -6.9471 73.4543>, rot=<-0.585 -0.6178 0.4677 0.2394>}
 dyn_belt_bk_08_null {type=Null, loc=<-19.466 -1.8965 88.1968>, rot=<0.5887 0.459 -0.4065 -0.5268>}
 dyn_belt_bk_08_ctrl {controlType=EULER_TRANSFORM, parent=dyn_belt_bk_08_null, type=Control, loc=<-63.0022 -4.0755 66.5448>, rot=<-0.6543 -0.5918 0.2807 0.3781>}
]
variables [
 RBAN Enabled {type=bool, default=True}
 belt_bk_ctrls {
  type="TArray<FRigElementKey>"
  default="((Type=Control,Name=\"dyn_belt_bk_01_ctrl\"),(Type=Control,Name=\"dyn_belt_bk_02_ctrl\"),(Type=Control,Name=\"dyn_belt_bk_03_ctrl\"),(Type=Control,Name=\"dyn_belt_bk_04_ctrl\"),(Type=Control,Name=\"dyn_belt_bk_05_ctrl\"),(Type=Control,Name=\"dyn_belt_bk_06_ctrl\"),(Ty ...(truncated)"
 }
 belt_bk_ctrls_RBAN {type="TArray<FRigElementKey>"}
 belt_bk_joints {
  type="TArray<FRigElementKey>"
  default="((Type=Bone,Name=\"dyn_belt_bk_01\"),(Type=Bone,Name=\"dyn_belt_bk_02\"),(Type=Bone,Name=\"dyn_belt_bk_03\"),(Type=Bone,Name=\"dyn_belt_bk_04\"),(Type=Bone,Name=\"dyn_belt_bk_05\"),(Type=Bone,Name=\"dyn_belt_bk_06\"),(Type=Bone,Name=\"dyn_belt_bk_07\"),(Type=Bone,Name=\" ...(truncated)"
 }
 belt_bk_joints_RBAN {type="TArray<FRigElementKey>"}
 belt_ctrls {
  type="TArray<FRigElementKey>"
  default="((Type=Control,Name=\"dyn_belt_01_ctrl\"),(Type=Control,Name=\"dyn_belt_02_ctrl\"),(Type=Control,Name=\"dyn_belt_03_ctrl\"),(Type=Control,Name=\"dyn_belt_04_ctrl\"),(Type=Control,Name=\"dyn_belt_05_ctrl\"),(Type=Control,Name=\"dyn_belt_06_ctrl\"),(Type=Control,Name=\"d ...(truncated)"
 }
 belt_ctrls_RBAN {type="TArray<FRigElementKey>"}
 belt_joints {
  type="TArray<FRigElementKey>"
  default=((Type=Bone,Name="dyn_belt_01"),(Type=Bone,Name="dyn_belt_02"),(Type=Bone,Name="dyn_belt_03"),(Type=Bone,Name="dyn_belt_04"),(Type=Bone,Name="dyn_belt_05"),(Type=Bone,Name="dyn_belt_06"),(Type=Bone,Name="dyn_belt_07"),(Type=Bone,Name="dyn_belt_08"))
 }
 belt_joints_RBAN {type="TArray<FRigElementKey>"}
 isInFrontEnd {type=double, default=0}
 isSkydiving {type=bool, default=True}
 shorts offset {type=FVector, default=(X=-6,Y=0,Z=0)}
]
