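# Physics asset for the left leg of the CorkFloorSwim shorts on the medium male player body, setting up secondary motion for the garment bones.
# Four dyn_shorts chains (front, back, inner, outer) each run master, 01, 02, with a capsule per bone and uniform tight swing and twist limits between links.
# Pelvis, left thigh and left calf capsules provide body volumes, and no constraint joins them to the chains or links one chain to another.
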PhysicsAsset /BRCosmetics/Characters/Player/Male/Medium/Bodies/M_MED_CorkFloorSwim/Meshes/M_MED_CorkFloorSwim_Shorts_L_Physics
bodies {
 calf_l [{shapeType=Capsule, length=53.4757, radius=7.5042, shapeName=calf_l_capsule, center=<-19.5928 -2.3895 0>, rotation=<rot 90 0 0>}]
 dyn_shorts_bk_01_l [{shapeType=Capsule, length=28.3735, radius=6.3398, shapeName=dyn_shorts_bk_01_l_capsule, center=<7.2688 -6.162 0>}]
 dyn_shorts_bk_02_l [{shapeType=Capsule, length=28.7027, radius=6.0106, shapeName=dyn_shorts_bk_02_l_capsule, center=<13.5068 -7.5782 0>}]
 dyn_shorts_bk_mstr_l [{shapeType=Capsule, length=20, radius=4, shapeName=dyn_shorts_bk_mstr_l_capsule}]
 dyn_shorts_fr_01_l [{shapeType=Capsule, length=28.3735, radius=6.3398, shapeName=dyn_shorts_fr_01_l_capsule, center=<7.2688 -6.162 0>}]
 dyn_shorts_fr_02_l [{shapeType=Capsule, length=28.3735, radius=6.3398, shapeName=dyn_shorts_fr_02_l_capsule, center=<13.5068 -6.162 0>}]
 dyn_shorts_fr_mstr_l [{shapeType=Capsule, length=0.505, radius=0.505, shapeName=dyn_shorts_fr_mstr_l_capsule, rotation=<rot 0 -90 0>}]
 dyn_shorts_in_01_l [{shapeType=Capsule, length=28.3735, radius=6.3398, shapeName=dyn_shorts_in_01_l_capsule, center=<7.2688 -6.162 0>}]
 dyn_shorts_in_02_l [{shapeType=Capsule, length=28.3735, radius=6.3398, shapeName=dyn_shorts_in_02_l_capsule, center=<13.5068 -6.162 0>}]
 dyn_shorts_in_mstr_l [{shapeType=Capsule, length=20, radius=4, shapeName=dyn_shorts_in_mstr_l_capsule}]
 dyn_shorts_out_01_l [{shapeType=Capsule, length=28.3735, radius=6.3398, shapeName=dyn_shorts_out_01_l_capsule, center=<7.2688 -6.162 0>}]
 dyn_shorts_out_02_l [{shapeType=Capsule, length=28.3735, radius=6.3398, shapeName=dyn_shorts_out_02_l_capsule, center=<13.5068 -6.162 0>}]
 dyn_shorts_out_mstr_l [{shapeType=Capsule, length=19.0061, radius=4.9939, shapeName=dyn_shorts_out_mstr_l_capsule}]
 pelvis [{shapeType=Capsule, length=4.2293, radius=16.374, shapeName=pelvis_capsule, center=<2.2183 0 0>, rotation=<rot 90 0 0>}]
 thigh_l [{shapeType=Capsule, length=72.2456, radius=9.5707, shapeName=thigh_l_capsule, center=<-32.6791 -1.2261 0>, rotation=<rot 90 0 0>}]
}
constraints [
 {bone1Name=dyn_shorts_bk_01_l, bone2Name=dyn_shorts_bk_mstr_l, swing1LimitDegrees=8, swing2LimitDegrees=8, twistLimitDegrees=2}
 {bone1Name=dyn_shorts_bk_02_l, bone2Name=dyn_shorts_bk_01_l, swing1LimitDegrees=8, swing2LimitDegrees=8, twistLimitDegrees=2}
 {bone1Name=dyn_shorts_fr_02_l, bone2Name=dyn_shorts_fr_01_l, swing1LimitDegrees=8, swing2LimitDegrees=8, twistLimitDegrees=2}
 {bone1Name=dyn_shorts_in_01_l, bone2Name=dyn_shorts_in_mstr_l, swing1LimitDegrees=8, swing2LimitDegrees=8, twistLimitDegrees=2}
 {bone1Name=dyn_shorts_in_02_l, bone2Name=dyn_shorts_in_01_l, swing1LimitDegrees=8, swing2LimitDegrees=8, twistLimitDegrees=2}
 {bone1Name=dyn_shorts_out_01_l, bone2Name=dyn_shorts_out_mstr_l, swing1LimitDegrees=8, swing2LimitDegrees=8, twistLimitDegrees=2}
 {bone1Name=dyn_shorts_out_02_l, bone2Name=dyn_shorts_out_01_l, swing1LimitDegrees=8, swing2LimitDegrees=8, twistLimitDegrees=2}
 {bone1Name=dyn_shorts_fr_01_l, bone2Name=dyn_shorts_fr_mstr_l, swing1LimitDegrees=8, swing2LimitDegrees=8, twistLimitDegrees=2}
]
