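# Physics asset for the right leg of the shorts on the male medium CorkFloorSwim body cosmetic, defining bodies for cloth-like bone motion.
# Four chains (front, back, inner, outer) each link a master and two segment capsules through constraints with tight swing and twist limits.
# Pelvis and right thigh capsules have no constraints, as body volumes beside the chains; the thigh shape carries a left-side name, thigh_l_capsule.
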PhysicsAsset /BRCosmetics/Characters/Player/Male/Medium/Bodies/M_MED_CorkFloorSwim/Meshes/M_MED_CorkFloorSwim_Shorts_R_Physics
bodies {
 dyn_shorts_bk_01_r [{shapeType=Capsule, length=29.7473, radius=6.33, shapeName=dyn_shorts_bk_01_r_capsule, center=<7.2688 -4.9863 0>}]
 dyn_shorts_bk_02_r [{shapeType=Capsule, length=31.1337, radius=6.33, shapeName=dyn_shorts_bk_02_r_capsule, center=<13.5068 -6.4354 0>}]
 dyn_shorts_bk_mstr_r [{shapeType=Capsule, length=28, radius=6.68, shapeName=dyn_shorts_bk_mstr_r_capsule}]
 dyn_shorts_fr_01_r [{shapeType=Capsule, length=29.7473, radius=6.33, shapeName=dyn_shorts_fr_01_r_capsule, center=<7.2688 -4.9863 0>}]
 dyn_shorts_fr_02_r [{shapeType=Capsule, length=29.7473, radius=6.33, shapeName=dyn_shorts_fr_02_r_capsule, center=<13.5068 -4.9411 0>}]
 dyn_shorts_fr_mstr_r [{shapeType=Capsule, length=28, radius=6.68, shapeName=dyn_shorts_fr_mstr_r_capsule}]
 dyn_shorts_in_01_r [{shapeType=Capsule, length=29.7473, radius=6.33, shapeName=dyn_shorts_in_01_r_capsule, center=<7.2688 -4.9863 0>}]
 dyn_shorts_in_02_r [{shapeType=Capsule, length=29.7473, radius=6.33, shapeName=dyn_shorts_in_02_r_capsule, center=<13.5068 -4.9411 0>}]
 dyn_shorts_in_mstr_r [{shapeType=Capsule, length=28, radius=6.68, shapeName=dyn_shorts_in_mstr_r_capsule}]
 dyn_shorts_out_01_r [{shapeType=Capsule, length=29.7473, radius=6.33, shapeName=dyn_shorts_out_01_r_capsule, center=<7.2688 -4.9863 0>}]
 dyn_shorts_out_02_r [{shapeType=Capsule, length=29.7473, radius=6.33, shapeName=dyn_shorts_out_02_r_capsule, center=<13.5068 -4.9411 0>}]
 dyn_shorts_out_mstr_r [{shapeType=Capsule, length=28, radius=6.68, shapeName=dyn_shorts_out_mstr_r_capsule}]
 pelvis [{shapeType=Capsule, length=4.2293, radius=16.374, shapeName=pelvis_capsule, center=<2.2183 0 0>, rotation=<rot 90 0 0>}]
 thigh_r [{shapeType=Capsule, length=70.1103, radius=8.0651, shapeName=thigh_l_capsule, center=<32.6791 1.2261 0>, rotation=<rot 90 0 0>}]
}
constraints [
 {bone1Name=dyn_shorts_bk_01_r, bone2Name=dyn_shorts_bk_mstr_r, swing1LimitDegrees=8, swing2LimitDegrees=8, twistLimitDegrees=2}
 {bone1Name=dyn_shorts_bk_02_r, bone2Name=dyn_shorts_bk_01_r, swing1LimitDegrees=8, swing2LimitDegrees=8, twistLimitDegrees=2}
 {bone1Name=dyn_shorts_fr_01_r, bone2Name=dyn_shorts_fr_mstr_r, swing1LimitDegrees=8, swing2LimitDegrees=8, twistLimitDegrees=2}
 {bone1Name=dyn_shorts_fr_02_r, bone2Name=dyn_shorts_fr_01_r, swing1LimitDegrees=8, swing2LimitDegrees=8, twistLimitDegrees=2}
 {bone1Name=dyn_shorts_in_01_r, bone2Name=dyn_shorts_in_mstr_r, swing1LimitDegrees=8, swing2LimitDegrees=8, twistLimitDegrees=2}
 {bone1Name=dyn_shorts_in_02_r, bone2Name=dyn_shorts_in_01_r, swing1LimitDegrees=8, swing2LimitDegrees=8, twistLimitDegrees=2}
 {bone1Name=dyn_shorts_out_01_r, bone2Name=dyn_shorts_out_mstr_r, swing1LimitDegrees=8, swing2LimitDegrees=8, twistLimitDegrees=2}
 {bone1Name=dyn_shorts_out_02_r, bone2Name=dyn_shorts_out_01_r, swing1LimitDegrees=8, swing2LimitDegrees=8, twistLimitDegrees=2}
]
