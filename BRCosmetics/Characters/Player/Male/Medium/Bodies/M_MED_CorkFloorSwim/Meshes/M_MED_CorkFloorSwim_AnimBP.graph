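# Post-copy animation Blueprint for the M_MED_CorkFloorSwim male body cosmetic, adding secondary motion to its belt and shorts.
# The attached parent's pose runs through a Control Rig, then RigidBody simulations for belt and each shorts leg.
# Chained spring-based, wind-enabled Anim Dynamics nodes then swing the shorts' flap bones.
# Each update sets shared gravity via a deprecated state function, lowers physics alpha while skydiving, and sets the rig's front-end input.
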AnimBlueprint /BRCosmetics/Characters/Player/Male/Medium/Bodies/M_MED_CorkFloorSwim/Meshes/M_MED_CorkFloorSwim_AnimBP
target_skeleton=&/M_MED_CorkFloorSwim_Skeleton
graphs [
 AnimGraph {nodes [
   n1 = Control Rig!(
    Source=Copy Pose From Mesh() {bUseAttachedParent=true}
    Alpha
    RBAN Enabled:bool=Get bRigidBodyEnabled()
    isSkydiving:bool=Get bIsSkyDiving()
    isInFrontEnd:double=Get isInFrontEnd() -> (isInFrontEnd:double)
   ) -> (Pose) {
    alphaBoolBlend {blendOption=Linear}
    controlRigAssetReference {blueprintRigClass=&/M_MED_CorkFloorSwim_CtrlRig1.M_MED_CorkFloorSwim_CtrlRig1_C}
   }
   n2 = Get Gravity Override()
   n3 = Get Skydive Alpha() -> (Skydive Alpha:double)
   n4 = RigidBody!(ComponentPose=Local To Component(LocalPose=@n1), Alpha=@n3, OverrideWorldGravity:Vector=@n2, ExternalForce) -> (Pose) {
    bClampLinearTranslationLimitToRefPose=true
    bOverrideWorldGravity=true
    overridePhysicsAsset=&/M_MED_CorkFloorSwim_Belt_Physics1
    simulationSpace=BaseBoneSpace
    alphaBoolBlend {blendOption=Linear}
    baseBoneRef {boneName=pelvis}
    componentAppliedLinearAccClamp=<3000 3000 3000>
    componentLinearAccScale=<0.4 0.4 0.7>
    componentLinearVelScale=<0.5 0.5 0.8>
    simSpaceSettings {maxAngularAcceleration=10, maxAngularVelocity=10, maxLinearAcceleration=100, maxLinearVelocity=100, worldAlpha=0.5}
   }
   n5 = RigidBody(ComponentPose=@n4, Alpha=@n3, OverrideWorldGravity=@n2) {
    bClampLinearTranslationLimitToRefPose=true
    bOverrideWorldGravity=true
    overridePhysicsAsset=&/M_MED_CorkFloorSwim_Shorts_L_Physics
    simulationSpace=BaseBoneSpace
    alphaBoolBlend {blendOption=Linear}
    baseBoneRef {boneName=thigh_l}
    componentAppliedLinearAccClamp=<5000 5000 5000>
    componentLinearAccScale=<0.4 0.4 0.4>
    componentLinearVelScale=<0.5 0.5 0.5>
    simSpaceSettings {maxAngularAcceleration=100, maxAngularVelocity=10, maxLinearAcceleration=300, maxLinearVelocity=300, worldAlpha=0.7}
   }
   n6 = RigidBody(ComponentPose=@n5, Alpha=@n3, OverrideWorldGravity=@n2) {
    bClampLinearTranslationLimitToRefPose=true
    bOverrideWorldGravity=true
    overridePhysicsAsset=&/M_MED_CorkFloorSwim_Shorts_R_Physics
    simulationSpace=BaseBoneSpace
    alphaBoolBlend {blendOption=Linear}
    baseBoneRef {boneName=thigh_r}
    componentAppliedLinearAccClamp=<5000 5000 5000>
    componentLinearAccScale=<0.4 0.4 0.4>
    componentLinearVelScale=<0.5 0.5 0.5>
    simSpaceSettings {maxAngularAcceleration=100, maxAngularVelocity=10, maxLinearAcceleration=300, maxLinearVelocity=300, worldAlpha=0.7}
   }
   n7 = Anim Dynamics - Bone: dyn_shorts_fr_flap_01_l!(ComponentPose=@n6, Alpha=@n3, GravityOverride:Vector=@n2, ExternalForce="1,0,0") -> (Pose) {
    angularDampingOverride=0.7
    angularSpringConstant=200
    bAngularSpring=true
    bEnableWind=true
    bOverrideAngularDamping=true
    bUseGravityOverride=true
    simulationSpace=BoneRelative
    windScale=5000
    alphaBoolBlend {blendOption=Linear}
    boundBone {boneName=dyn_shorts_fr_flap_01_l}
    componentLinearAccScale=<0.5 0.5 0.5>
    componentLinearVelScale=<0.6 0.6 0.6>
    physicsBodyDefinitions [{
      boundBone {boneName=dyn_shorts_fr_flap_01_l}
      boxExtents=<9 9 9>
      constraintSetup {angularLimitsMax=<0 30 20>, angularLimitsMin=<0 -30 -20>, angularTarget=<1 0 0>}
      localJointOffset=<0.4 0 0>
     }]
    relativeSpaceBone {boneName=root}
   }
   n8 = Anim Dynamics - Bone: dyn_shorts_out_flap_02_l!(ComponentPose=@n7, Alpha=@n3, GravityOverride:Vector=@n2, ExternalForce="1,0,0") -> (Pose) {
    angularDampingOverride=0.7
    angularSpringConstant=200
    bAngularSpring=true
    bEnableWind=true
    bOverrideAngularDamping=true
    bUseGravityOverride=true
    simulationSpace=BoneRelative
    windScale=5000
    alphaBoolBlend {blendOption=Linear}
    boundBone {boneName=dyn_shorts_out_flap_02_l}
    componentLinearAccScale=<0.5 0.5 0.5>
    componentLinearVelScale=<0.6 0.6 0.6>
    physicsBodyDefinitions [{
      boundBone {boneName=dyn_shorts_out_flap_02_l}
      boxExtents=<9 9 9>
      constraintSetup {angularLimitsMax=<0 30 20>, angularLimitsMin=<0 -30 -20>, angularTarget=<1 0 0>}
      localJointOffset=<0.4 0 0>
     }]
    relativeSpaceBone {boneName=root}
   }
   n9 = Anim Dynamics - Bone: dyn_shorts_out_flap_01_l!(ComponentPose=@n8, Alpha=@n3, GravityOverride:Vector=@n2, ExternalForce="1,0,0") -> (Pose) {
    angularDampingOverride=0.7
    angularSpringConstant=200
    bAngularSpring=true
    bEnableWind=true
    bOverrideAngularDamping=true
    bUseGravityOverride=true
    simulationSpace=BoneRelative
    windScale=5000
    alphaBoolBlend {blendOption=Linear}
    boundBone {boneName=dyn_shorts_out_flap_01_l}
    componentLinearAccScale=<0.5 0.5 0.5>
    componentLinearVelScale=<0.6 0.6 0.6>
    physicsBodyDefinitions [{
      boundBone {boneName=dyn_shorts_out_flap_01_l}
      boxExtents=<8 8 8>
      constraintSetup {angularLimitsMax=<0 30 20>, angularLimitsMin=<0 -30 -20>, angularTarget=<1 0 0>}
      localJointOffset=<0.4 0 0>
     }]
    relativeSpaceBone {boneName=root}
   }
   n10 = Anim Dynamics - Bone: dyn_shorts_bk_flap_02_l!(ComponentPose=@n9, Alpha=@n3, GravityOverride:Vector=@n2, ExternalForce="1,0,0") -> (Pose) {
    angularDampingOverride=0.7
    angularSpringConstant=200
    bAngularSpring=true
    bEnableWind=true
    bOverrideAngularDamping=true
    bUseGravityOverride=true
    simulationSpace=BoneRelative
    windScale=5000
    alphaBoolBlend {blendOption=Linear}
    boundBone {boneName=dyn_shorts_bk_flap_02_l}
    componentLinearAccScale=<0.5 0.5 0.5>
    componentLinearVelScale=<0.6 0.6 0.6>
    physicsBodyDefinitions [{
      boundBone {boneName=dyn_shorts_bk_flap_02_l}
      boxExtents=<9 9 9>
      constraintSetup {angularLimitsMax=<0 30 20>, angularLimitsMin=<0 -30 -10>, angularTarget=<1 0 0>}
      localJointOffset=<0.4 0 0>
     }]
    relativeSpaceBone {boneName=root}
   }
   n11 = Anim Dynamics - Bone: dyn_shorts_bk_flap_01_l!(ComponentPose=@n10, Alpha=@n3, GravityOverride:Vector=@n2, ExternalForce="1,0,0") -> (Pose) {
    angularDampingOverride=0.7
    angularSpringConstant=200
    bAngularSpring=true
    bEnableWind=true
    bOverrideAngularDamping=true
    bUseGravityOverride=true
    simulationSpace=BoneRelative
    windScale=5000
    alphaBoolBlend {blendOption=Linear}
    boundBone {boneName=dyn_shorts_bk_flap_01_l}
    componentLinearAccScale=<0.5 0.5 0.5>
    componentLinearVelScale=<0.6 0.6 0.6>
    physicsBodyDefinitions [{
      boundBone {boneName=dyn_shorts_bk_flap_01_l}
      boxExtents=<9 9 9>
      constraintSetup {angularLimitsMax=<0 30 20>, angularLimitsMin=<0 -30 -10>, angularTarget=<1 0 0>}
      localJointOffset=<0.4 0 0>
     }]
    relativeSpaceBone {boneName=root}
   }
   n12 = Anim Dynamics - Bone: dyn_shorts_in_flap_02_l!(ComponentPose=@n11, Alpha=@n3, GravityOverride:Vector=@n2, ExternalForce="1,0,0") -> (Pose) {
    angularDampingOverride=0.7
    angularSpringConstant=200
    bAngularSpring=true
    bEnableWind=true
    bOverrideAngularDamping=true
    bUseGravityOverride=true
    simulationSpace=BoneRelative
    windScale=5000
    alphaBoolBlend {blendOption=Linear}
    boundBone {boneName=dyn_shorts_in_flap_02_l}
    componentLinearAccScale=<0.5 0.5 0.5>
    componentLinearVelScale=<0.6 0.6 0.6>
    physicsBodyDefinitions [{
      boundBone {boneName=dyn_shorts_in_flap_02_l}
      boxExtents=<9 9 9>
      constraintSetup {angularLimitsMax=<0 30 20>, angularLimitsMin=<0 -30 -20>, angularTarget=<1 0 0>}
      localJointOffset=<0.4 0 0>
     }]
    relativeSpaceBone {boneName=root}
   }
   n13 = Anim Dynamics - Bone: dyn_shorts_in_flap_01_l!(ComponentPose=@n12, Alpha=@n3, GravityOverride:Vector=@n2, ExternalForce="1,0,0") -> (Pose) {
    angularDampingOverride=0.7
    angularSpringConstant=200
    bAngularSpring=true
    bEnableWind=true
    bOverrideAngularDamping=true
    bUseGravityOverride=true
    simulationSpace=BoneRelative
    windScale=5000
    alphaBoolBlend {blendOption=Linear}
    boundBone {boneName=dyn_shorts_in_flap_01_l}
    componentLinearAccScale=<0.5 0.5 0.5>
    componentLinearVelScale=<0.6 0.6 0.6>
    physicsBodyDefinitions [{
      boundBone {boneName=dyn_shorts_in_flap_01_l}
      boxExtents=<9 9 9>
      constraintSetup {angularLimitsMax=<0 30 20>, angularLimitsMin=<0 -30 -20>, angularTarget=<1 0 0>}
      localJointOffset=<0.4 0 0>
     }]
    relativeSpaceBone {boneName=root}
   }
   n14 = Anim Dynamics - Bone: dyn_shorts_fr_flap_01_r!(ComponentPose=@n13, Alpha=@n3, GravityOverride:Vector=@n2, ExternalForce="1,0,0") -> (Pose) {
    angularDampingOverride=0.7
    angularSpringConstant=200
    bAngularSpring=true
    bEnableWind=true
    bOverrideAngularDamping=true
    bUseGravityOverride=true
    simulationSpace=BoneRelative
    windScale=5000
    alphaBoolBlend {blendOption=Linear}
    boundBone {boneName=dyn_shorts_fr_flap_01_r}
    componentLinearAccScale=<0.5 0.5 0.5>
    componentLinearVelScale=<0.6 0.6 0.6>
    physicsBodyDefinitions [{
      boundBone {boneName=dyn_shorts_fr_flap_01_r}
      boxExtents=<9 9 9>
      constraintSetup {angularLimitsMax=<0 30 20>, angularLimitsMin=<0 -30 -20>, angularTarget=<1 0 0>}
      localJointOffset=<0.4 0 0>
     }]
    relativeSpaceBone {boneName=root}
   }
   n15 = Anim Dynamics - Bone: dyn_shorts_out_flap_01_r!(ComponentPose=@n14, Alpha=@n3, GravityOverride:Vector=@n2, ExternalForce="1,0,0") -> (Pose) {
    angularDampingOverride=0.7
    angularSpringConstant=200
    bAngularSpring=true
    bEnableWind=true
    bOverrideAngularDamping=true
    bUseGravityOverride=true
    simulationSpace=BoneRelative
    windScale=5000
    alphaBoolBlend {blendOption=Linear}
    boundBone {boneName=dyn_shorts_out_flap_01_r}
    componentLinearAccScale=<0.5 0.5 0.5>
    componentLinearVelScale=<0.6 0.6 0.6>
    physicsBodyDefinitions [{
      boundBone {boneName=dyn_shorts_out_flap_01_r}
      boxExtents=<9 9 9>
      constraintSetup {angularLimitsMax=<0 30 20>, angularLimitsMin=<0 -30 -20>, angularTarget=<1 0 0>}
      localJointOffset=<0.4 0 0>
     }]
    relativeSpaceBone {boneName=root}
   }
   n16 = Anim Dynamics - Bone: dyn_shorts_bk_flap_01_r!(ComponentPose=@n15, Alpha=@n3, GravityOverride:Vector=@n2, ExternalForce="1,0,0") -> (Pose) {
    angularDampingOverride=0.7
    angularSpringConstant=200
    bAngularSpring=true
    bEnableWind=true
    bOverrideAngularDamping=true
    bUseGravityOverride=true
    simulationSpace=BoneRelative
    windScale=5000
    alphaBoolBlend {blendOption=Linear}
    boundBone {boneName=dyn_shorts_bk_flap_01_r}
    componentLinearAccScale=<0.5 0.5 0.5>
    componentLinearVelScale=<0.6 0.6 0.6>
    physicsBodyDefinitions [{
      boundBone {boneName=dyn_shorts_bk_flap_01_r}
      boxExtents=<9 9 9>
      constraintSetup {angularLimitsMax=<0 30 20>, angularLimitsMin=<0 -30 -10>, angularTarget=<1 0 0>}
      localJointOffset=<0.4 0 0>
     }]
    relativeSpaceBone {boneName=root}
   }
   n17 = Anim Dynamics - Bone: dyn_shorts_bk_flap_02_r!(ComponentPose=@n16, Alpha=@n3, GravityOverride:Vector=@n2, ExternalForce="1,0,0") -> (Pose) {
    angularDampingOverride=0.7
    angularSpringConstant=200
    bAngularSpring=true
    bEnableWind=true
    bOverrideAngularDamping=true
    bUseGravityOverride=true
    simulationSpace=BoneRelative
    windScale=5000
    alphaBoolBlend {blendOption=Linear}
    boundBone {boneName=dyn_shorts_bk_flap_02_r}
    componentLinearAccScale=<0.5 0.5 0.5>
    componentLinearVelScale=<0.6 0.6 0.6>
    physicsBodyDefinitions [{
      boundBone {boneName=dyn_shorts_bk_flap_02_r}
      boxExtents=<9 9 9>
      constraintSetup {angularLimitsMax=<0 30 20>, angularLimitsMin=<0 -30 -10>, angularTarget=<1 0 0>}
      localJointOffset=<0.4 0 0>
     }]
    relativeSpaceBone {boneName=root}
   }
   n18 = Anim Dynamics - Bone: dyn_shorts_in_flap_02_r!(ComponentPose=@n17, Alpha=@n3, GravityOverride:Vector=@n2, ExternalForce="1,0,0") -> (Pose) {
    angularDampingOverride=0.7
    angularSpringConstant=200
    bAngularSpring=true
    bEnableWind=true
    bOverrideAngularDamping=true
    bUseGravityOverride=true
    simulationSpace=BoneRelative
    windScale=5000
    alphaBoolBlend {blendOption=Linear}
    boundBone {boneName=dyn_shorts_in_flap_02_r}
    componentLinearAccScale=<0.5 0.5 0.5>
    componentLinearVelScale=<0.6 0.6 0.6>
    physicsBodyDefinitions [{
      boundBone {boneName=dyn_shorts_in_flap_02_r}
      boxExtents=<9 9 9>
      constraintSetup {angularLimitsMax=<0 30 20>, angularLimitsMin=<0 -30 -20>, angularTarget=<1 0 0>}
      localJointOffset=<0.4 0 0>
     }]
    relativeSpaceBone {boneName=root}
   }
   n19 = Anim Dynamics - Bone: dyn_shorts_in_flap_01_r!(ComponentPose=@n18, Alpha=@n3, GravityOverride:Vector=@n2, ExternalForce="1,0,0") -> (Pose) {
    angularDampingOverride=0.7
    angularSpringConstant=200
    bAngularSpring=true
    bEnableWind=true
    bOverrideAngularDamping=true
    bUseGravityOverride=true
    simulationSpace=BoneRelative
    windScale=5000
    alphaBoolBlend {blendOption=Linear}
    boundBone {boneName=dyn_shorts_in_flap_01_r}
    componentLinearAccScale=<0.5 0.5 0.5>
    componentLinearVelScale=<0.6 0.6 0.6>
    physicsBodyDefinitions [{
      boundBone {boneName=dyn_shorts_in_flap_01_r}
      boxExtents=<9 9 9>
      constraintSetup {angularLimitsMax=<0 30 20>, angularLimitsMin=<0 -30 -20>, angularTarget=<1 0 0>}
      localJointOffset=<0.4 0 0>
     }]
    relativeSpaceBone {boneName=root}
   }
   Output Pose(Result=Component To Local(ComponentPose=@n19))
  ]}
 EventGraph {nodes [
   n1 = Event BlueprintUpdateAnimation()
   > n2 = Sequence!(execute) -> (then_0, then_1, then_2:Exec)
   > n3 = Gravity Override State Function DEPRECATED(
    CharacterPartAnimInstance=Self-Reference()
    GameTimeInSeconds=@n1.DeltaTimeX
    GravityOverrideParams=Get Gravity Override Params()
    RBAN Params=Get RBAN Params()
   )
   > Set Gravity Override(Gravity Override=@n3)
   n4 = Branch(execute=@n2.then_1, Condition=Get bIsSkyDiving())
   > Set Skydive Alpha(Skydive Alpha:double=0.75) -> (Output_Get:double)
   Set Skydive Alpha(execute=@n4.else, Skydive Alpha=1)
   n5 = Branch(execute=@n2.then_2, Condition=Get bIsInFrontEnd())
   > Set isInFrontEnd(isInFrontEnd:double=1) -> (Output_Get:double)
   Set isInFrontEnd(execute=@n5.else, isInFrontEnd=0)
   TryGetPawnOwner()
  ]}
]
variables [
 Gravity Override {type=Vector}
 Gravity Override Params {type=GravityOverrideParamsStruct, default {skydive_FloatingDown {gravityOverride=<0 -700 -500>}}}
 isInFrontEnd {type=number}
 isTacSprinting {type=number}
 RBAN Params {
  type=RBANParamsStruct
  default {
   dBNO {simSpaceSettings {
     maxAngularAcceleration=1000
     maxAngularVelocity=3
     maxLinearAcceleration=300
     maxLinearVelocity=700
     velocityScaleZ=0
     externalLinearDragV=<0 0.5 0>
    }}
   dBNO_Carried {simSpaceSettings {
     maxAngularAcceleration=1000
     maxAngularVelocity=3
     maxLinearAcceleration=300
     maxLinearVelocity=700
     velocityScaleZ=0
     externalLinearDragV=<0 0.5 0>
    }}
   emoteOrMelee {simSpaceSettings {
     maxAngularAcceleration=1000
     maxAngularVelocity=3
     maxLinearAcceleration=300
     maxLinearVelocity=700
     velocityScaleZ=0
     externalLinearDragV=<0 0.5 0>
    }}
   front_End {simSpaceSettings {
     maxAngularAcceleration=1000
     maxAngularVelocity=3
     maxLinearAcceleration=300
     maxLinearVelocity=700
     velocityScaleZ=0
     externalLinearDragV=<0 0.5 0>
    }}
   onGround_Crouch_Moving {simSpaceSettings {
     maxAngularAcceleration=1000
     maxAngularVelocity=3
     maxLinearAcceleration=300
     maxLinearVelocity=700
     velocityScaleZ=0
     externalLinearDragV=<0 0.5 0>
    }}
   onGround_Falling {simSpaceSettings {
     maxAngularAcceleration=1000
     maxAngularVelocity=3
     maxLinearAcceleration=300
     maxLinearVelocity=700
     velocityScaleZ=0
     externalLinearDragV=<0 0.5 0>
    }}
   onGround_Moving {simSpaceSettings {
     maxAngularAcceleration=1000
     maxAngularVelocity=3
     maxLinearAcceleration=300
     maxLinearVelocity=700
     velocityScaleZ=0
     externalLinearDragV=<0 0.5 0>
    }}
   onGround_Standing {simSpaceSettings {
     maxAngularAcceleration=1000
     maxAngularVelocity=3
     maxLinearAcceleration=300
     maxLinearVelocity=700
     velocityScaleZ=0
     externalLinearDragV=<0 0.5 0>
    }}
   rail_Grinding {simSpaceSettings {
     maxAngularAcceleration=1000
     maxAngularVelocity=3
     maxLinearAcceleration=300
     maxLinearVelocity=700
     velocityScaleZ=0
     worldAlpha=0.1
     externalLinearDragV=<0 0.5 0>
    }}
   riding_Moving {simSpaceSettings {
     maxAngularAcceleration=1000
     maxAngularVelocity=3
     maxLinearAcceleration=300
     maxLinearVelocity=700
     velocityScaleZ=0
     externalLinearDragV=<0 0.5 0>
    }}
   riding_Moving_Targeting {simSpaceSettings {
     maxAngularAcceleration=1000
     maxAngularVelocity=3
     maxLinearAcceleration=300
     maxLinearVelocity=700
     velocityScaleZ=0
     externalLinearDragV=<0 0.5 0>
    }}
   riding_Standing {simSpaceSettings {
     maxAngularAcceleration=1000
     maxAngularVelocity=3
     maxLinearAcceleration=300
     maxLinearVelocity=700
     velocityScaleZ=0
     externalLinearDragV=<0 0.5 0>
    }}
   skydive_DivingDown {simSpaceSettings {
     maxAngularAcceleration=1000
     maxAngularVelocity=3
     maxLinearAcceleration=300
     maxLinearVelocity=700
     velocityScaleZ=0
     externalLinearDragV=<0 0.5 0>
    }}
   skydive_FloatingDown {
    componentLinearAccScale=<0 0.5 0>
    componentLinearVelScale=<0 0.5 0>
    simSpaceSettings {
     maxAngularAcceleration=1000
     maxAngularVelocity=3
     maxLinearAcceleration=300
     maxLinearVelocity=700
     velocityScaleZ=0
     externalLinearDragV=<0 0.5 0>
    }
   }
   skydive_GoingUp {simSpaceSettings {
     maxAngularAcceleration=1000
     maxAngularVelocity=3
     maxLinearAcceleration=300
     maxLinearVelocity=700
     velocityScaleZ=0
     externalLinearDragV=<0 0.5 0>
    }}
   skydive_Parachuting {
    componentLinearAccScale=<0 0.1 0.3>
    componentLinearVelScale=<0 0.1 0.3>
    simSpaceSettings {
     maxAngularAcceleration=1000
     maxAngularVelocity=3
     maxLinearAcceleration=300
     maxLinearVelocity=700
     velocityScaleZ=0
     externalLinearDragV=<0 0.5 0>
    }
   }
   skydive_Surfing {simSpaceSettings {
     maxAngularAcceleration=1000
     maxAngularVelocity=3
     maxLinearAcceleration=300
     maxLinearVelocity=700
     velocityScaleZ=0
     externalLinearDragV=<0 0.5 0>
    }}
   swimming {simSpaceSettings {
     maxAngularAcceleration=1000
     maxAngularVelocity=3
     maxLinearAcceleration=300
     maxLinearVelocity=700
     velocityScaleZ=0
     externalLinearDragV=<0 0.5 0>
    }}
   tactical_Sprint {simSpaceSettings {
     maxAngularAcceleration=1000
     maxAngularVelocity=3
     maxLinearAcceleration=300
     maxLinearVelocity=700
     velocityScaleZ=0
     externalLinearDragV=<0 0.5 0>
    }}
   ziplining {simSpaceSettings {
     maxAngularAcceleration=1000
     maxAngularVelocity=3
     maxLinearAcceleration=300
     maxLinearVelocity=700
     velocityScaleZ=0
     externalLinearDragV=<0 0.5 0>
    }}
  }
 }
 Skydive Alpha {type=number}
]
