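# Physics asset for the CorkFloorSwim male medium body's face accessory, setting up four dynamic braids (front and back, left and right).
# Each braid is a four-link capsule chain hanging from a small master-bone body, joined by swing and twist limited constraints that loosen toward the tip while twist stays tight.
# Unconstrained capsules on head, upper spine, clavicles and upper arms provide body volumes, likely for the braids to collide against.
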PhysicsAsset /BRCosmetics/Characters/Player/Male/Medium/Bodies/M_MED_CorkFloorSwim/Meshes/Parts/M_MED_CorkFloorSwim_FaceAcc_Physics1
bodies {
 clavicle_l [{
   shapeType=Capsule
   length=19.7691
   radius=6.2309
   shapeName=clavicle_l_capsule
   center=<21.3156 -1.2091 0.729>
   rotation=<rot 72.928 -0.0298 -23.1892>
  }]
 clavicle_r [{
   shapeType=Capsule
   length=14.4677
   radius=11.5323
   shapeName=clavicle_r_capsule
   center=<-8.7816 -0.6468 0.1922>
   rotation=<rot 68.996 -0.0003 -0.0004>
  }]
 dyn_braid_bk_01_l [{
   shapeType=Capsule
   length=7.0246
   radius=1.0754
   shapeName=dyn_braid_bk_01_l_capsule
   center=<5.0988 -1.0873 0.5607>
   rotation=<rot 84.3081 10.7863 -0.0001>
  }]
 dyn_braid_bk_01_r [{
   shapeType=Capsule
   length=6.8689
   radius=1.2311
   shapeName=dyn_braid_bk_01_r_capsule
   center=<5.8797 -0.9682 0>
   rotation=<rot 90 0 -13.3742>
  }]
 dyn_braid_bk_02_l [{
   shapeType=Capsule
   length=6.7633
   radius=1.2641
   shapeName=dyn_braid_bk_02_r_capsule
   center=<4.8373 -0.029 0.018>
   rotation=<rot 54.5711 89.9999 87.7817>
  }]
 dyn_braid_bk_02_r [{
   shapeType=Capsule
   length=6.4271
   radius=0.8079
   shapeName=dyn_braid_bk_02_r_capsule
   center=<4.5665 0.0765 0.1663>
   rotation=<rot 54.5711 89.9999 89.9999>
  }]
 dyn_braid_bk_03_l [{
   shapeType=Capsule
   length=5.6922
   radius=0.8152
   shapeName=dyn_braid_bk_03_r_capsule
   center=<3.3593 0.0201 0.062>
   rotation=<rot 51.5903 -95.1414 -93.3327>
  }]
 dyn_braid_bk_03_r [{
   shapeType=Capsule
   length=5.1785
   radius=0.7202
   shapeName=dyn_braid_bk_03_r_capsule
   center=<3.467 0.0787 0.0618>
   rotation=<rot 75.0663 76.6086 73.692>
  }]
 dyn_braid_bk_04_l [{
   shapeType=Capsule
   length=2.828
   radius=1.472
   shapeName=dyn_braid_bk_04_r_capsule
   center=<2.907 0.1583 -0.313>
   rotation=<rot 48.7847 -77.0167 -80.1606>
  }]
 dyn_braid_bk_04_r [{
   shapeType=Capsule
   length=2.8244
   radius=1.4756
   shapeName=dyn_braid_bk_04_r_capsule
   center=<2.8881 -0.2837 -0.6642>
   rotation=<rot -59.591 -90.0001 -89.9999>
  }]
 dyn_braid_bk_mstr_l [{shapeType=Capsule, length=0.505, radius=0.505, shapeName=dyn_braid_bk_mstr_l_capsule, rotation=<rot 90 0 0>}]
 dyn_braid_bk_mstr_r [{shapeType=Capsule, length=0.505, radius=0.505, shapeName=dyn_braid_bk_mstr_r_capsule, rotation=<rot 90 0 0>}]
 dyn_braid_fr_01_l [{
   shapeType=Capsule
   length=6.8689
   radius=1.2311
   shapeName=dyn_braid_fr_01_l_capsule
   center=<4.2854 -0.3394 0>
   rotation=<rot 90 180 174.5626>
  }]
 dyn_braid_fr_01_r [{
   shapeType=Capsule
   length=6.9258
   radius=1.1742
   shapeName=dyn_braid_fr_01_r_capsule
   center=<4.605 -0.1756 0>
   rotation=<rot 90 180 176.6964>
  }]
 dyn_braid_fr_02_l [{shapeType=Capsule, length=7.7093, radius=0.923, shapeName=dyn_braid_fr_02_r_capsule, center=<4.1635 0 0.0039>, rotation=<rot 90 0 0>}]
 dyn_braid_fr_02_r [{
   shapeType=Capsule
   length=7.5838
   radius=0.913
   shapeName=dyn_braid_fr_02_r_capsule
   center=<4.5256 0 0.1526>
   rotation=<rot 89.9175 -0.0007 -0.0007>
  }]
 dyn_braid_fr_03_l [{
   shapeType=Capsule
   length=4.5552
   radius=0.6982
   shapeName=dyn_braid_fr_03_r_capsule
   center=<2.94 -0.0187 0.2061>
   rotation=<rot 51.7498 -88.3389 -88.6954>
  }]
 dyn_braid_fr_03_r [{
   shapeType=Capsule
   length=4.491
   radius=0.7623
   shapeName=dyn_braid_fr_03_r_capsule
   center=<2.94 -0.0117 -0.0771>
   rotation=<rot -40.6437 -90.0039 -89.9973>
  }]
 dyn_braid_fr_04_l [{
   shapeType=Capsule
   length=4.0807
   radius=1.7614
   shapeName=dyn_braid_fr_04_r_capsule
   center=<2.7976 -0.3949 0.3827>
   rotation=<rot 43.6268 -82.4389 -84.7683>
  }]
 dyn_braid_fr_04_r [{
   shapeType=Capsule
   length=4.0771
   radius=1.7651
   shapeName=dyn_braid_fr_04_r_capsule
   center=<2.8314 -0.4686 -0.5623>
   rotation=<rot -61.1036 -90.0001 -89.9999>
  }]
 dyn_braid_fr_mstr_l [{shapeType=Capsule, length=0.505, radius=0.505, shapeName=dyn_braid_fr_mstr_l_capsule, rotation=<rot 90 0 0>}]
 dyn_braid_fr_mstr_r [{shapeType=Capsule, length=0.505, radius=0.505, shapeName=dyn_braid_fr_mstr_r_capsule, rotation=<rot 90 0 0>}]
 head [{shapeType=Capsule, length=12.5079, radius=7.0418, shapeName=head_capsule, center=<3.3359 1.3318 0>, rotation=<rot 90 0 0>}]
 spine_04 [{shapeType=Capsule, length=54.0854, radius=18.1097, shapeName=neck_02_capsule, center=<28.2431 -12.8369 0>}]
 spine_05 [{shapeType=Capsule, length=38.6893, radius=18.6359, shapeName=spine_05_capsule, center=<-5.8791 0.2887 0>, rotation=<rot 0 5.3724 0>}]
 upperarm_l [{
   shapeType=Capsule
   length=31.544
   radius=6.1085
   shapeName=upperarm_l_capsule
   center=<5.4676 0.0243 3.7347>
   rotation=<rot 37.1486 -90.0465 -90>
  }]
 upperarm_r [{shapeType=Capsule, length=24.1991, radius=10.9341, shapeName=upperarm_r_capsule, center=<-5.3096 0.7939 -2.0183>, rotation=<rot 90 0 0>}]
}
constraints [
 {bone1Name=dyn_braid_bk_01_l, bone2Name=dyn_braid_bk_mstr_l, swing1LimitDegrees=10, swing2LimitDegrees=10, twistLimitDegrees=2}
 {bone1Name=dyn_braid_bk_02_l, bone2Name=dyn_braid_bk_01_l, swing1LimitDegrees=12, swing2LimitDegrees=12, twistLimitDegrees=2}
 {bone1Name=dyn_braid_bk_03_l, bone2Name=dyn_braid_bk_02_l, swing1LimitDegrees=12, swing2LimitDegrees=12, twistLimitDegrees=2}
 {bone1Name=dyn_braid_bk_04_l, bone2Name=dyn_braid_bk_03_l, swing1LimitDegrees=26, swing2LimitDegrees=26, twistLimitDegrees=2}
 {bone1Name=dyn_braid_bk_01_r, bone2Name=dyn_braid_bk_mstr_r, swing1LimitDegrees=10, swing2LimitDegrees=10, twistLimitDegrees=2}
 {bone1Name=dyn_braid_bk_02_r, bone2Name=dyn_braid_bk_01_r, swing1LimitDegrees=12, swing2LimitDegrees=12, twistLimitDegrees=2}
 {bone1Name=dyn_braid_bk_03_r, bone2Name=dyn_braid_bk_02_r, swing1LimitDegrees=12, swing2LimitDegrees=12, twistLimitDegrees=2}
 {bone1Name=dyn_braid_bk_04_r, bone2Name=dyn_braid_bk_03_r, swing1LimitDegrees=26, swing2LimitDegrees=26, twistLimitDegrees=2}
 {bone1Name=dyn_braid_fr_01_l, bone2Name=dyn_braid_fr_mstr_l, swing1LimitDegrees=10, swing2LimitDegrees=10, twistLimitDegrees=2}
 {bone1Name=dyn_braid_fr_02_l, bone2Name=dyn_braid_fr_01_l, swing1LimitDegrees=12, swing2LimitDegrees=12, twistLimitDegrees=2}
 {bone1Name=dyn_braid_fr_03_l, bone2Name=dyn_braid_fr_02_l, swing1LimitDegrees=12, swing2LimitDegrees=12, twistLimitDegrees=2}
 {bone1Name=dyn_braid_fr_04_l, bone2Name=dyn_braid_fr_03_l, swing1LimitDegrees=26, swing2LimitDegrees=26, twistLimitDegrees=2}
 {bone1Name=dyn_braid_fr_01_r, bone2Name=dyn_braid_fr_mstr_r, swing1LimitDegrees=10, swing2LimitDegrees=10, twistLimitDegrees=2}
 {bone1Name=dyn_braid_fr_02_r, bone2Name=dyn_braid_fr_01_r, swing1LimitDegrees=12, swing2LimitDegrees=12, twistLimitDegrees=2}
 {bone1Name=dyn_braid_fr_03_r, bone2Name=dyn_braid_fr_02_r, swing1LimitDegrees=12, swing2LimitDegrees=12, twistLimitDegrees=2}
 {bone1Name=dyn_braid_fr_04_r, bone2Name=dyn_braid_fr_03_r, swing1LimitDegrees=26, swing2LimitDegrees=26, twistLimitDegrees=2}
]
